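# One-shot stereo MetaSound source for a UI danger popup sound, apparently played on entering The Hive zone.
# On Play picks a random wave from five popup variations and triggers a stereo Wave Player, which feeds both outputs and signals On Finished.
# It uses the Fort_UI sound class and overrides concurrency to a single instance with a retrigger time to limit stacking.
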MetaSoundSource /TheHivePOIUI/Assets/Sound/MSS_UI_Enter_TheHive_Popup
inputs [
 UE.Source.OnPlay {type=Trigger}
 Waves {type=WaveAsset:Array}
]
interfaces [UE.OutputFormat.Stereo:1.0, UE.Source.OneShot:1.0, UE.Source:1.1]
outputs [
 "UE.OutputFormat.Stereo.Audio:0" {type=Audio}
 "UE.OutputFormat.Stereo.Audio:1" {type=Audio}
 UE.Source.OneShot.OnFinished {type=Trigger}
]
pages [Default {
  input_defaults {
   UE.Source.OnPlay=null
   Waves [
    &/UI_HiveZone_Danger_Popup_05
    &/UI_HiveZone_Danger_Popup_04
    &/UI_HiveZone_Danger_Popup_03
    &/UI_HiveZone_Danger_Popup_02
    &/UI_HiveZone_Danger_Popup_01
   ]
  }
  nodes [
   ArrayRandomGet_0 = Array.Random Get:WaveAsset:Array(Next=^UE.Source.OnPlay, In Array=^Waves)
   WavePlayer_0 = UE.Wave Player:Stereo(Play=@ArrayRandomGet_0.On Next, Wave Asset=@ArrayRandomGet_0.Value) {also {version [1,1]}}
  ]
  links [
   {dst="@UE.Source.OneShot.OnFinished", src=WavePlayer_0.On Finished}
   {dst="@UE.OutputFormat.Stereo.Audio:0", src=WavePlayer_0.Out Left}
   {dst="@UE.OutputFormat.Stereo.Audio:1", src=WavePlayer_0.Out Right}
  ]
 }]
properties {
 assetImportData=&self:AssetImportData
 bOverrideConcurrency=true
 numChannels=2
 outputFormat=Stereo
 sampleRate=48000
 soundClassObject=&Fort_UI
 concurrencyOverrides {maxCount=1, retriggerTime=2}
}
