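# UI material that draws the Hive POI enter icon: an exclamation mark with triangle, bracket and glow backer.
# Exclamation and triangle come from distance-field textures, while bracket and glow are built from SDF shapes and rotated UV edges, all antialiased.
# High Level blends grey toward orange, and scalar parameters drive intro, flash, scale, bracket width and gradient reveal.
# Panning noise distorts UVs for glitching, and weighted masks combine through Max for alpha-composite opacity.
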Material /TheHivePOIUI/UI/Materials/M_UI_HivePOIEnter_Icon
graph {
 properties {bAutomaticallySetUsageInEditor=false, blendMode=BLEND_AlphaComposite, materialDomain=MD_UI, usedShadingModels=Default Lit}
 nodes [
  Constant3Vector() {constant=<rgba 0.24 0.165 0.12 0>}
  Constant()
  n1 = NamedRerouteUsage()
  n2 = Multiply(
   A=LinearInterpolate(
    A=Constant3Vector() {constant=<rgba 0.3464 0.3464 0.3464 1>}
    B=Constant3Vector() {constant=<rgba 1 0.1976 0 1>}
    Alpha=ScalarParameter() {parameterName=High Level}
   )
   B=Multiply(A=NamedRerouteUsage()) {constB=0.2}
  )
  n3 = Constant3Vector() {constant=<rgba 0.1484 0.1484 0.1484 1>}
  n4 = Constant3Vector() {constant=<rgba 1 0.1356 0 1>}
  n5 = ScalarParameter() {parameterName=High Level}
  LinearInterpolate(
   A=LinearInterpolate(A=Add(A=@n1, B=@n2), B=LinearInterpolate(A=@n3, B=@n4, Alpha=@n5), Alpha=NamedRerouteUsage())
   B=Constant3Vector() {constant=<rgba 1 0.1356 0 1>}
   Alpha=NamedRerouteUsage()
  )
  MaterialFunctionCall() {materialFunction=&GetUserInterfaceUV}
  MaterialFunctionCall() {materialFunction=&DebugTimeSine}
  MaterialFunctionCall() {materialFunction=&GetUserInterfaceUV}
  n6 = Multiply(A=NamedRerouteUsage()) {constB=0.8}
  n7 = Multiply(A=NamedRerouteUsage()) {constB=0}
  n8 = NamedRerouteUsage()
  Max(
   A=Max(A=Max(A=@n6, B=@n7), B=Multiply(A=@n8) {constB=0.5})
   B=Multiply(A=NamedRerouteUsage(), B=ScalarParameter() {defaultValue=0.8, parameterName="! Translucency"}) {constB=0.8}
  )
  Multiply(B=ScalarParameter() {defaultValue=1, parameterName="! Opacity"})
  n9 = Constant2Vector() {g=0.27}
  n10 = Add(A=NamedRerouteUsage(), B=Add(A=@n9, B=Constant2Vector() {g=-0.02}))
  n11 = Add(A=LinearInterpolate(Alpha=ScalarParameter() {parameterName="! Scale"}) {constA=0.35, constAlpha=0, constB=0.7}, B=Constant() {r=-0.1})
  n12 = TextureSample(UVs=MaterialFunctionCall(UV=MaterialFunctionCall(UVs=@n10, Texture Scale=@n11) {materialFunction=&ScaleUVsByCenter}.UVs) {materialFunction=&MF_UI_UVScaling}.Scaled to Fit) {samplerType=SAMPLERTYPE_LinearColor, texture=/TheHivePOIUI/UI/Textures/T_UI_Icon_Notification_Exclamation}
  n13 = Constant() {r=0.8}
  n14 = MaterialFunctionCall(DistanceField=@n12.RGB, Threshold=@n13) {materialFunction=&MF_UI_AntiAliasing}
  n15 = MaterialFunctionCall(DistanceField=@n12.RGB, Threshold=Add(A=@n13) {constB=-0.4}) {materialFunction=&MF_UI_AntiAliasing}
  NamedRerouteDeclaration(None=Max(A=NamedRerouteDeclaration(None=@n14.Outer) {name="!icon", nodeColor=<rgba 0.75 0.6063 0.375 0.75>}, B=@n15.Outer)) {name="!Icon+Outline", nodeColor=<rgba 0.75 0.6063 0.375 0.75>}
  n16 = NamedRerouteUsage()
  n17 = Constant3Vector() {constant=<rgba 1 0.1356 0 1>}
  n18 = ScalarParameter() {parameterName=TriangleFlash}
  n19 = LinearInterpolate(
   A=LinearInterpolate(A=LinearInterpolate(Alpha=@n16) {constA=0.4}, Alpha=NamedRerouteUsage()) {constA=1, constB=0.2}
   B=LinearInterpolate(A=LinearInterpolate(A=@n17, Alpha=@n18), B=Constant3Vector() {constant=<rgba 0.2502 0.1144 0.0369 1>}, Alpha=NamedRerouteUsage())
   Alpha=ScalarParameter() {parameterName=High Level}
  )
  n20 = NamedRerouteUsage()
  n21 = NamedRerouteUsage()
  n22 = NamedRerouteUsage()
  n23 = LinearInterpolate(
   A=LinearInterpolate(B=LinearInterpolate(Alpha=@n21) {constA=0.4}, Alpha=@n22)
   B=LinearInterpolate(Alpha=@n22) {constA=1, constB=0}
   Alpha=ScalarParameter() {parameterName="Intro ! Switch"}
  )
  n24 = LinearInterpolate(A=Constant3Vector() {constant=<rgba 0.0865 0.0395 0.0242 1>}, B=Constant3Vector() {constant=<rgba 1 0.2881 0 1>}, Alpha=NamedRerouteUsage())
  n25 = ScalarParameter() {parameterName=High Level}
  NamedRerouteDeclaration(None=LinearInterpolate(A=LinearInterpolate(B=@n19, Alpha=@n20), B=LinearInterpolate(A=@n23, B=@n24, Alpha=@n25), Alpha=NamedRerouteUsage())) {name=ColorIcon, nodeColor=<rgba 0 0 0 1>}
  n26 = ScalarParameter() {parameterName=Speed Up}
  n27 = Panner(
   Coordinate=TextureCoordinate() {uTiling=4, vTiling=2}
   Time=Multiply(A=Time(), B=LinearInterpolate(Alpha=@n26) {constA=0.2, constB=3})
   Speed=Constant2Vector() {g=0.25}
  )
  n28 = Subtract(A=ComponentMask(None=TextureSample(UVs=@n27) {samplerType=SAMPLERTYPE_LinearGrayscale, texture=/TheHivePOIUI/UI/Textures/T_UI_POI_Enter_Noise}.RGB) {g=true, r=true}) {constB=0.5}
  n29 = Multiply(A=ScalarParameter() {parameterName=Glitch In}) {constB=0.1}
  NamedRerouteDeclaration(None=Add(A=NamedRerouteUsage(), B=Multiply(A=@n28, B=@n29))) {name=Glitch UVs, nodeColor=<rgba 0 0 0 1>}
  NamedRerouteDeclaration(None=MaterialFunctionCall() {materialFunction=&GetUserInterfaceUV}.Normalized UV) {name=Normalized UV, nodeColor=<rgba 0 0 0 1>}
  n30 = Constant2Vector() {desc=IconGlowOffset, g=0.125}
  n31 = Constant2Vector() {g=-0.225}
  n32 = MaterialFunctionCall() {materialFunction=&GetUserInterfaceUV}
  n33 = ConstantBiasScale(None=Add(A=Add(A=@n30, B=@n31), B=@n32.Normalized UV)) {bias=-0.5, scale=2}
  n34 = Constant()
  n35 = Constant() {r=0.05}
  n36 = MaterialFunctionCall(UVs=Multiply(A=@n33) {constB=-1}, Rotation Center=@n34, "Rotation Angle (0-1)"=@n35) {materialFunction=&CustomRotator}
  n37 = MaterialFunctionCall(UVs=@n33, Rotation Center=@n34, "Rotation Angle (0-1)"=Multiply(A=@n35) {constB=-1}) {materialFunction=&CustomRotator}
  n38 = Max(A=ComponentMask(None=@n36.Rotated Values) {r=true}, B=ComponentMask(None=@n37.Rotated Values) {r=true})
  n39 = Constant2Vector() {g=0.3}
  n40 = MaterialFunctionCall!(UVs=Add(A=Add(A=@n39, B=@n30), B=@n32.Normalized UV), Size=Constant2Vector() {g=0.6, r=0.19}, CornerRoundness, KeepAspectRatio, PixelSize) {materialFunction=/FortUILibrary/MaterialFunctions/Shapes/SDFs/MF_SDF_Box}
  n41 = Saturate(None=MaterialFunctionCall(DistanceField=Max(A=@n38, B=@n40.SDF)) {materialFunction=&MF_UI_AntiAliasing}.Inner)
  n42 = Power(Base=Multiply(A=ComponentMask(None=@n32.Normalized UV) {g=true}) {constB=2}) {constExponent=3}
  n43 = Power(
   Base=Multiply(A=NamedRerouteDeclaration(None=@n41) {name=Mask_GlowBacker, nodeColor=<rgba 0.75 0.6063 0.375 0.75>}, B=Multiply(A=@n42) {constB=4})
   Exp=LinearInterpolate(Alpha=ScalarParameter() {parameterName=Gradient Front Reveal}) {constA=1, constB=3}
  )
  n44 = ScalarParameter() {defaultValue=1, parameterName=Gradient Opacity}
  NamedRerouteDeclaration(None=Saturate(None=Multiply(A=@n43, B=@n44))) {name=OpacityMask_GlowBacker, nodeColor=<rgba 0.75 0.6063 0.375 0.75>}
  NamedRerouteDeclaration(None=Multiply(A=@n43, B=ScalarParameter() {parameterName=Gradient Front Intensity})) {name=Glow, nodeColor=<rgba 1 0 0.4922 1>}
  n45 = LinearInterpolate(Alpha=ScalarParameter() {defaultValue=1, parameterName=Bracket Show}) {constAlpha=1, constB=0.3}
  n46 = Constant() {r=1}
  n47 = MaterialFunctionCall(DistanceField=MaterialFunctionCall(Size=AppendVector(A=@n45, B=@n46)) {materialFunction=&MF_UI_Box}.SDF) {materialFunction=&MF_UI_AntiAliasing}
  n48 = Add(A=MaterialFunctionCall() {materialFunction=&GetUserInterfaceUV}.Normalized UV, B=@n30)
  n49 = Add(A=Constant2Vector() {g=0.068}, B=@n48)
  n50 = Constant3Vector() {constant=<rgba 0.27 0.16 0.17 0>}
  n51 = Add(A=@n48, B=Constant2Vector() {g=0.03})
  n52 = Subtract(A=@n50.R) {constB=0.01}
  n53 = Add(A=@n50.G) {constB=0.02}
  n54 = Subtract(A=@n50.B) {constB=0.04}
  n55 = Max(
   A=Multiply(A=MaterialFunctionCall!(UVs=@n49, TopWidth=@n50.R, BottomWidth=@n50.R, Height=@n50.R) {materialFunction=/FortUILibrary/MaterialFunctions/Shapes/SDFs/MF_SDF_IsoscelesTrapezoid}.SDF) {constB=-1}
   B=Multiply(A=MaterialFunctionCall(UVs=@n51, TopWidth=@n52, BottomWidth=@n53, Height=@n54) {materialFunction=&MF_SDF_IsoscelesTrapezoid}.SDF)
  )
  n56 = Multiply(A=@n37.Rotated Values) {constB=-1}
  n57 = Add(A=Max(A=ComponentMask(None=@n36.Rotated Values) {g=true}, B=ComponentMask(None=@n56) {g=true})) {constB=-0.58}
  n58 = Add(A=TextureCoordinate(), B=Constant2Vector() {g=0.26})
  n59 = Constant2Vector() {g=0.2, r=1}
  n60 = Multiply(
   A=MaterialFunctionCall(DistanceField=Max(A=@n55, B=@n57)) {materialFunction=&MF_UI_AntiAliasing}.Inner
   B=OneMinus(None=MaterialFunctionCall(UVs=@n58, Size=@n59) {materialFunction=&MF_UI_Box}.Fill)
  )
  NamedRerouteDeclaration(None=Saturate(None=Multiply(A=@n47.Inner, B=@n60))) {name=Mask_Bracket, nodeColor=<rgba 0.75 0.6063 0.375 0.75>}
  n61 = Add(A=NamedRerouteUsage(), B=@n9)
  n62 = ScalarParameter() {defaultValue=0.35, parameterName=Triangle Scale}
  n63 = TextureSample(UVs=MaterialFunctionCall(UV=MaterialFunctionCall(UVs=@n61, Texture Scale=@n62) {materialFunction=&ScaleUVsByCenter}.UVs) {materialFunction=&MF_UI_UVScaling}.Scaled to Fit) {samplerType=SAMPLERTYPE_LinearColor, texture=/TheHivePOIUI/UI/Textures/T_UI_Icon_Notification_Bg}
  n64 = Constant() {r=0.75}
  n65 = MaterialFunctionCall(DistanceField=@n63.RGB, Threshold=@n64) {materialFunction=&MF_UI_AntiAliasing}
  n66 = ScalarParameter() {defaultValue=1, parameterName=Triangle Master Opacity}
  n67 = NamedRerouteDeclaration(None=Multiply(A=@n65.Outer, B=@n66) {constB=0}) {name=TriangleIcon, nodeColor=<rgba 0.75 0.6063 0.375 0.75>}
  n68 = MaterialFunctionCall(DistanceField=@n63.RGB, Threshold=Add(A=@n64) {constB=-0.15}) {materialFunction=&MF_UI_AntiAliasing}
  n69 = NamedRerouteDeclaration(None=Multiply(A=Multiply(A=@n65.Inner, B=@n68.Outer), B=@n66) {constB=0}) {name=TriangleOutline, nodeColor=<rgba 0.75 0.6063 0.375 0.75>}
  NamedRerouteDeclaration(None=Add(A=Multiply(A=@n67) {constB=0.4}, B=Multiply(A=@n69) {constB=0.8})) {name=TriangleIcon+Outline, nodeColor=<rgba 0.75 0.6063 0.375 0.75>}
  NamedRerouteDeclaration(None=Multiply(A=@n14.Inner, B=@n15.Outer)) {name="!Outline", nodeColor=<rgba 0.75 0.6063 0.375 0.75>}
  NamedRerouteDeclaration(None=MaterialFunctionCall(Speed=Constant() {r=1.5}) {materialFunction=&DebugTimeSine}.Result) {name=FlashTime, nodeColor=<rgba 0 0 0 1>}
  ScalarParameter() {parameterName=High-Bracket}
 ]
}
