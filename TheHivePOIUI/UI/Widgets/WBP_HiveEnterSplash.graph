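# Priority HUD splash announcing entry into the Hive bug-biome POI, with its map name and a High or Low threat level.
# It listens for the biome enter/exit stateful event on the current view target, rebinding when that changes, and reads threat and POI tag from the boss spawn manager.
# Becoming active plays sound and a threat-specific, material-driven intro and starts a display timer; the outro's end collapses the widget.
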WidgetBlueprint /TheHivePOIUI/UI/Widgets/WBP_HiveEnterSplash
animations [
 Anim_Intro_LowThreat {
  binding_count=11
  length_seconds=1.45
  object=&self:Anim_Intro_LowThreat
  sequence {
   bindings [
    Image_Icon {
     id=#1
     possessed_class=Image
     tracks [
      {
       type=MovieSceneFloatTrack
       display_name=Render Opacity
       sections [{type=MovieSceneFloatSection, channels [None {type=MovieSceneScriptingFloatChannel, keys [[0.002,1],[22.001,0],[44.001,1]]}]}]
      }
      {
       type=MovieSceneWidgetMaterialTrack
       sections [{
         type=MovieSceneParameterSection
         end_frame=87
         start_frame=0
         channels [
          Glitch In {
           type=MovieSceneScriptingFloatChannel
           thinned_from=22
           keys [[0,0],[6.002,0],[12.002,0.11],[13.002,0],[17.001,0],[18.001,-0.1],[21.001,0],[22.001,0.23],[26.001,0],[27.001,-0.11],[31.001,0],[62.002,0],[68.002,0.11],[69.002,0],[73.001,0],[74.001,-0.1],[77.001,0],[78.001,0.23],[82.001,0],[83.001,-0.11],[87.001,0]]
          }
          Bracket Show {type=MovieSceneScriptingFloatChannel, keys [[10.002,0],[25.001,1]]}
          Speed Up {type=MovieSceneScriptingFloatChannel, keys [[0.002,1],[4.002,0]]}
          Gradient Front Reveal {type=MovieSceneScriptingFloatChannel, keys [[9.002,1],[16.002,0]]}
          Triangle Scale {type=MovieSceneScriptingFloatChannel, keys [[0.002,0],[9.002,0.35]]}
          "! Scale" {type=MovieSceneScriptingFloatChannel, keys [[0.002,0.9],[9.002,0]]}
          Gradient Front Intensity {type=MovieSceneScriptingFloatChannel, keys [[9.002,0],[19.001,0.9],[28.001,0]]}
          Gradient Opacity {type=MovieSceneScriptingFloatChannel, keys [[9.002,0],[14.002,1]]}
          "Intro ! Switch" {type=MovieSceneScriptingFloatChannel, keys [[0,1],[9.002,0]]}
          "! Translucency" {type=MovieSceneScriptingFloatChannel, keys [[0,0],[9.002,0.8]]}
          Triangle Master Opacity {type=MovieSceneScriptingFloatChannel, keys [[0,0],[9.002,1]]}
         ]
        }]
      }
     ]
    }
    Image_ThreatBacker {
     id=#2
     possessed_class=Image
     tracks [{
       type=MovieSceneWidgetMaterialTrack
       sections [{
         type=MovieSceneParameterSection
         end_frame=87
         start_frame=0
         channels [
          Outline Thickness {type=MovieSceneScriptingFloatChannel, const [1,1]}
          Outline Master Opacity {type=MovieSceneScriptingFloatChannel, keys [[25,0],[42,1]]}
          Inner Shape Master Opacity {type=MovieSceneScriptingFloatChannel, keys [[0,0],[31,0],[39,0.4]]}
          LowHighSwitch {type=MovieSceneScriptingFloatChannel, const [0,1]}
         ]
        }]
      }]
    }
    Overlay_30 {
     id=#3
     possessed_class=Overlay
     tracks [{
       type=MovieSceneFloatTrack
       display_name=Render Opacity
       sections [{type=MovieSceneFloatSection, end_frame=0, start_frame=0, channels [None {type=MovieSceneScriptingFloatChannel, const [1,1]}]}]
      }]
    }
    Overlay_2 {
     id=#4
     possessed_class=Overlay
     tracks [{
       type=MovieSceneFloatTrack
       display_name=Render Opacity
       sections [{type=MovieSceneFloatSection, end_frame=0, start_frame=0, channels [None {type=MovieSceneScriptingFloatChannel, const [1,1]}]}]
      }]
    }
    Image_Glow {
     id=#5
     possessed_class=Image
     tracks [
      {
       type=MovieSceneFloatTrack
       display_name=Render Opacity
       sections [{type=MovieSceneFloatSection, channels [None {type=MovieSceneScriptingFloatChannel, keys [[5,0],[16,0.304],[27,0]]}]}]
      }
      {
       type=MovieScene2DTransformTrack
       sections [{
         type=MovieScene2DTransformSection
         end_frame=27
         start_frame=5
         channels [
          Translation.X {type=MovieSceneScriptingFloatChannel, const [0,1]}
          Translation.Y {type=MovieSceneScriptingFloatChannel, keys [[5,-40],[27,-80]]}
          Angle {type=MovieSceneScriptingFloatChannel, const [0,1]}
          Scale.X {type=MovieSceneScriptingFloatChannel, const [1,1]}
          Scale.Y {type=MovieSceneScriptingFloatChannel, keys [[5,1],[13,1.7896]]}
          Shear.X {type=MovieSceneScriptingFloatChannel, const [0,1]}
          Shear.Y {type=MovieSceneScriptingFloatChannel, const [0,1]}
         ]
        }]
      }
     ]
    }
    Image_TextFx {
     id=#6
     possessed_class=Image
     tracks [
      {
       type=MovieSceneWidgetMaterialTrack
       sections [{
         type=MovieSceneParameterSection
         channels [
          Blinky X Movement {type=MovieSceneScriptingFloatChannel, keys [[26.001,0.266],[34.999,-0.28]]}
          Mask {type=MovieSceneScriptingFloatChannel, keys [[26.001,0],[35.001,1]]}
         ]
        }]
      }
      {
       type=MovieSceneFloatTrack
       display_name=Render Opacity
       sections [{type=MovieSceneFloatSection, channels [None {type=MovieSceneScriptingFloatChannel, thinned_from=6, keys [[15,0],[17,1],[41,1],[44,0]]}]}]
      }
     ]
    }
    Image_Backer {
     id=#7
     possessed_class=Image
     tracks [
      {
       type=MovieSceneFloatTrack
       display_name=Render Opacity
       sections [{type=MovieSceneFloatSection, channels [None {type=MovieSceneScriptingFloatChannel, const [1,1]}]}]
      }
      {
       type=MovieSceneWidgetMaterialTrack
       sections [{
         type=MovieSceneParameterSection
         end_frame=87
         start_frame=0
         channels [
          SignBacker_Opacity {type=MovieSceneScriptingFloatChannel, const [0.4,1]}
          Outline_Opacity {type=MovieSceneScriptingFloatChannel, keys [[22,0],[31,2],[64,0.2]]}
          Dots Opacity {type=MovieSceneScriptingFloatChannel, default=0, keys [[0,0],[1,1],[3,0],[7,1],[9,0.8],[13,0],[18,1]]}
          Stripes Opacity {type=MovieSceneScriptingFloatChannel, const [0,1]}
          Background Reveal {type=MovieSceneScriptingFloatChannel, keys [[11,0],[38,1]]}
          Outline Reveal {type=MovieSceneScriptingFloatChannel, keys [[6,0],[54,1]]}
          Backer_Opacity {type=MovieSceneScriptingFloatChannel, keys [[0,0],[68,0.8]]}
          Dots Scale {type=MovieSceneScriptingFloatChannel, thinned_from=4, keys [[0,1],[5,1],[6,0]]}
          Bottom Outline Reveal {type=MovieSceneScriptingFloatChannel, keys [[7,1],[49,0]]}
         ]
        }]
      }
     ]
    }
    SizeBox_1 {
     id=#8
     possessed_class=SizeBox
     tracks [{
       type=MovieSceneFloatTrack
       display_name=Render Opacity
       sections [{
         type=MovieSceneFloatSection
         channels [None {type=MovieSceneScriptingFloatChannel, keys [[32.485,0],[36.8,1],[39.758,0],[41.697,1],[44,0],[44.8,1],[46.546,0],[48,1]]}]
        }]
      }]
    }
    Overlay_0 {
     id=#9
     possessed_class=Overlay
     tracks [{
       type=MovieSceneFloatTrack
       display_name=Render Opacity
       sections [{type=MovieSceneFloatSection, end_frame=0, start_frame=0, channels [None {type=MovieSceneScriptingFloatChannel, const [1,1]}]}]
      }]
    }
    Text_POIName {
     id=#10
     possessed_class=CommonTextBlock
     tracks [
      {
       type=MovieSceneFloatTrack
       display_name=Render Opacity
       sections [{type=MovieSceneFloatSection, channels [None {type=MovieSceneScriptingFloatChannel, thinned_from=4, keys [[0,0],[31.001,0],[32.997,1]]}]}]
      }
      {
       type=MovieScene2DTransformTrack
       sections [{
         type=MovieScene2DTransformSection
         channels [
          Translation.X {type=MovieSceneScriptingFloatChannel, keys [[21.999,-30],[33.001,0],[41,0]]}
          Translation.Y {type=MovieSceneScriptingFloatChannel, const [5,1]}
          Angle {type=MovieSceneScriptingFloatChannel, const [0,1]}
          Scale.X {type=MovieSceneScriptingFloatChannel, const [1,1]}
          Scale.Y {type=MovieSceneScriptingFloatChannel, const [1,1]}
          Shear.X {type=MovieSceneScriptingFloatChannel, const [0,1]}
          Shear.Y {type=MovieSceneScriptingFloatChannel, const [0,1]}
         ]
        }]
      }
     ]
    }
    ThreatLevelDescriptionText {
     id=#11
     possessed_class=CommonTextBlock
     tracks [{
       type=MovieSceneColorTrack
       display_name=Color and Opacity
       sections [{
         type=MovieSceneColorSection
         channels [
          Color.R {type=MovieSceneScriptingFloatChannel, const [1,2]}
          Color.G {type=MovieSceneScriptingFloatChannel, const [0.2241,2]}
          Color.B {type=MovieSceneScriptingFloatChannel, const [0,2]}
          Color.A {type=MovieSceneScriptingFloatChannel, const [1,2]}
         ]
        }]
      }]
    }
   ]
   display_rate {numerator=60}
  }
 }
 Anim_Intro_HighThreat {
  binding_count=12
  length_seconds=1.45
  object=&self:Anim_Intro_HighThreat
  sequence {
   bindings [
    Overlay_30 {
     id=#12
     possessed_class=Overlay
     tracks [{
       type=MovieSceneFloatTrack
       display_name=Render Opacity
       sections [{type=MovieSceneFloatSection, end_frame=0, start_frame=0, channels [None {type=MovieSceneScriptingFloatChannel, const [1,1]}]}]
      }]
    }
    Image_TextFx {
     id=#13
     possessed_class=Image
     tracks [
      {
       type=MovieSceneWidgetMaterialTrack
       sections [{
         type=MovieSceneParameterSection
         channels [
          Blinky X Movement {type=MovieSceneScriptingFloatChannel, keys [[26.001,0.266],[34.999,-0.28]]}
          Mask {type=MovieSceneScriptingFloatChannel, keys [[26.001,0],[35.001,1]]}
         ]
        }]
      }
      {
       type=MovieSceneFloatTrack
       display_name=Render Opacity
       sections [{type=MovieSceneFloatSection, channels [None {type=MovieSceneScriptingFloatChannel, thinned_from=6, keys [[15,0],[17,1],[41,1],[44,0]]}]}]
      }
     ]
    }
    Overlay_0 {
     id=#14
     possessed_class=Overlay
     tracks [{
       type=MovieSceneFloatTrack
       display_name=Render Opacity
       sections [{type=MovieSceneFloatSection, end_frame=0, start_frame=0, channels [None {type=MovieSceneScriptingFloatChannel, const [1,1]}]}]
      }]
    }
    Image_Glow {
     id=#15
     possessed_class=Image
     tracks [
      {
       type=MovieSceneFloatTrack
       display_name=Render Opacity
       sections [{type=MovieSceneFloatSection, channels [None {type=MovieSceneScriptingFloatChannel, keys [[5,0],[16,0.304],[27,0]]}]}]
      }
      {
       type=MovieScene2DTransformTrack
       sections [{
         type=MovieScene2DTransformSection
         end_frame=27
         start_frame=5
         channels [
          Translation.X {type=MovieSceneScriptingFloatChannel, const [0,1]}
          Translation.Y {type=MovieSceneScriptingFloatChannel, keys [[5,-40],[27,-80]]}
          Angle {type=MovieSceneScriptingFloatChannel, const [0,1]}
          Scale.X {type=MovieSceneScriptingFloatChannel, const [1,1]}
          Scale.Y {type=MovieSceneScriptingFloatChannel, keys [[5,1],[12,1.7896]]}
          Shear.X {type=MovieSceneScriptingFloatChannel, const [0,1]}
          Shear.Y {type=MovieSceneScriptingFloatChannel, const [0,1]}
         ]
        }]
      }
     ]
    }
    SizeBox_1 {
     id=#16
     possessed_class=SizeBox
     tracks [{
       type=MovieSceneFloatTrack
       display_name=Render Opacity
       sections [{
         type=MovieSceneFloatSection
         channels [None {type=MovieSceneScriptingFloatChannel, keys [[32.485,0],[36.8,1],[39.758,0],[41.697,1],[44,0],[44.8,1],[46.546,0],[48,1]]}]
        }]
      }]
    }
    Image_ThreatBacker {
     id=#17
     possessed_class=Image
     tracks [{
       type=MovieSceneWidgetMaterialTrack
       sections [{
         type=MovieSceneParameterSection
         end_frame=87
         start_frame=0
         channels [
          Outline Thickness {type=MovieSceneScriptingFloatChannel, const [1,1]}
          Outline Master Opacity {type=MovieSceneScriptingFloatChannel, keys [[25,0],[42,1]]}
          Inner Shape Master Opacity {type=MovieSceneScriptingFloatChannel, keys [[0,0],[31,0],[39,0.4]]}
          LowHighSwitch {type=MovieSceneScriptingFloatChannel, keys [[24,0],[38,1]]}
         ]
        }]
      }]
    }
    Overlay_2 {
     id=#18
     possessed_class=Overlay
     tracks [{
       type=MovieSceneFloatTrack
       display_name=Render Opacity
       sections [{type=MovieSceneFloatSection, end_frame=0, start_frame=0, channels [None {type=MovieSceneScriptingFloatChannel, const [1,1]}]}]
      }]
    }
    Text_POIName {
     id=#19
     possessed_class=CommonTextBlock
     tracks [
      {
       type=MovieSceneFloatTrack
       display_name=Render Opacity
       sections [{type=MovieSceneFloatSection, channels [None {type=MovieSceneScriptingFloatChannel, thinned_from=4, keys [[0,0],[31.001,0],[32.997,1]]}]}]
      }
      {
       type=MovieScene2DTransformTrack
       sections [{
         type=MovieScene2DTransformSection
         channels [
          Translation.X {type=MovieSceneScriptingFloatChannel, keys [[21.999,-30],[33.001,0],[41,0]]}
          Translation.Y {type=MovieSceneScriptingFloatChannel, const [5,1]}
          Angle {type=MovieSceneScriptingFloatChannel, const [0,1]}
          Scale.X {type=MovieSceneScriptingFloatChannel, const [1,1]}
          Scale.Y {type=MovieSceneScriptingFloatChannel, const [1,1]}
          Shear.X {type=MovieSceneScriptingFloatChannel, const [0,1]}
          Shear.Y {type=MovieSceneScriptingFloatChannel, const [0,1]}
         ]
        }]
      }
     ]
    }
    Image_Icon {
     id=#20
     possessed_class=Image
     tracks [
      {
       type=MovieSceneWidgetMaterialTrack
       sections [{
         type=MovieSceneParameterSection
         end_frame=65
         start_frame=0
         channels [
          Glitch In {
           type=MovieSceneScriptingFloatChannel
           keys [[0,0],[6.002,0],[12.002,0.11],[13.002,0],[17.001,0],[18.001,-0.1],[21.001,0],[22.001,0.23],[26.001,0],[27.001,-0.11],[31.001,0]]
          }
          Bracket Show {type=MovieSceneScriptingFloatChannel, keys [[10.002,0],[25.001,1]]}
          Speed Up {type=MovieSceneScriptingFloatChannel, keys [[0.002,1],[4.002,0]]}
          Gradient Front Reveal {type=MovieSceneScriptingFloatChannel, keys [[9.002,1],[16.002,0]]}
          Triangle Scale {type=MovieSceneScriptingFloatChannel, keys [[0.002,0],[9.002,0.35]]}
          "! Scale" {type=MovieSceneScriptingFloatChannel, keys [[0.002,0.9],[9.002,0]]}
          Gradient Front Intensity {type=MovieSceneScriptingFloatChannel, keys [[9.002,0],[19.001,0.9],[28.001,0]]}
          Gradient Opacity {type=MovieSceneScriptingFloatChannel, keys [[9.002,0],[14.002,1]]}
          "Intro ! Switch" {type=MovieSceneScriptingFloatChannel, keys [[0,1],[9.002,0]]}
          "! Translucency" {type=MovieSceneScriptingFloatChannel, keys [[0,0],[9.002,0.8]]}
          Triangle Master Opacity {type=MovieSceneScriptingFloatChannel, keys [[0,0],[9.002,1]]}
          High Level {type=MovieSceneScriptingFloatChannel, const [1,1]}
         ]
        }]
      }
      {
       type=MovieSceneFloatTrack
       display_name=Render Opacity
       sections [{type=MovieSceneFloatSection, channels [None {type=MovieSceneScriptingFloatChannel, keys [[0.002,1],[22.001,0],[44.001,1]]}]}]
      }
     ]
    }
    ThreatLevelDescriptionText {
     id=#21
     possessed_class=CommonTextBlock
     tracks [{
       type=MovieSceneColorTrack
       display_name=Color and Opacity
       sections [{
         type=MovieSceneColorSection
         channels [
          Color.R {type=MovieSceneScriptingFloatChannel, keys [[24,1],[38,0]]}
          Color.G {type=MovieSceneScriptingFloatChannel, keys [[24,0.2346],[38,0]]}
          Color.B {type=MovieSceneScriptingFloatChannel, const [0,2]}
          Color.A {type=MovieSceneScriptingFloatChannel, const [1,2]}
         ]
        }]
      }]
    }
    Image_HighThreatTextGlow {
     id=#22
     possessed_class=Image
     tracks [
      {
       type=MovieSceneFloatTrack
       display_name=Render Opacity
       sections [{type=MovieSceneFloatSection, channels [None {type=MovieSceneScriptingFloatChannel, keys [[33,0],[41.001,0.5],[52,0.06]]}]}]
      }
      {
       type=MovieScene2DTransformTrack
       sections [{
         type=MovieScene2DTransformSection
         channels [
          Translation.X {type=MovieSceneScriptingFloatChannel, const [10,2]}
          Translation.Y {type=MovieSceneScriptingFloatChannel, const [0,1]}
          Angle {type=MovieSceneScriptingFloatChannel, const [90,1]}
          Scale.X {type=MovieSceneScriptingFloatChannel, const [0.15,1]}
          Scale.Y {type=MovieSceneScriptingFloatChannel, keys [[33,10],[41,15],[52,10]]}
          Shear.X {type=MovieSceneScriptingFloatChannel, const [0,1]}
          Shear.Y {type=MovieSceneScriptingFloatChannel, const [0,1]}
         ]
        }]
      }
     ]
    }
    Image_Backer {
     id=#23
     possessed_class=Image
     tracks [
      {
       type=MovieSceneWidgetMaterialTrack
       sections [{
         type=MovieSceneParameterSection
         end_frame=87
         start_frame=0
         channels [
          SignBacker_Opacity {type=MovieSceneScriptingFloatChannel, const [0.4,1]}
          Outline_Opacity {type=MovieSceneScriptingFloatChannel, keys [[22,0],[31,2],[64,0.2]]}
          Dots Opacity {type=MovieSceneScriptingFloatChannel, default=0, keys [[0,0],[1,1],[3,0],[7,1],[9,0.8],[13,0],[18,1]]}
          Stripes Opacity {type=MovieSceneScriptingFloatChannel, const [0,1]}
          Background Reveal {type=MovieSceneScriptingFloatChannel, keys [[11,0],[38,1]]}
          Outline Reveal {type=MovieSceneScriptingFloatChannel, keys [[6,0],[54,1]]}
          Backer_Opacity {type=MovieSceneScriptingFloatChannel, keys [[0,0],[68,0.8]]}
          Dots Scale {type=MovieSceneScriptingFloatChannel, thinned_from=4, keys [[0,1],[5,1],[6,0]]}
          Bottom Outline Reveal {type=MovieSceneScriptingFloatChannel, keys [[7,1],[49,0]]}
         ]
        }]
      }
      {
       type=MovieSceneFloatTrack
       display_name=Render Opacity
       sections [{type=MovieSceneFloatSection, channels [None {type=MovieSceneScriptingFloatChannel, const [1,1]}]}]
      }
     ]
    }
   ]
   display_rate {numerator=60}
  }
 }
 Anim_Outro {
  binding_count=2
  length_seconds=0.35
  object=&self:Anim_Outro
  sequence {bindings [
    Overlay_0 {
     id=#24
     possessed_class=Overlay
     tracks [{
       type=MovieSceneFloatTrack
       display_name=Render Opacity
       sections [{type=MovieSceneFloatSection, end_frame=4, start_frame=0, channels [None {type=MovieSceneScriptingFloatChannel, keys [[0,1],[4,0]]}]}]
      }]
    }
    Overlay_2 {
     id=#25
     possessed_class=Overlay
     tracks [{
       type=MovieSceneFloatTrack
       display_name=Render Opacity
       sections [{type=MovieSceneFloatSection, end_frame=7, start_frame=3, channels [None {type=MovieSceneScriptingFloatChannel, keys [[3,1],[7,0]]}]}]
      }]
    }
   ]}
 }
]
graphs [
 EventGraph {nodes [
   n1 = DeactivatePriorityWidget() {is=title K2Node_CustomEvent}
   > SetVisibility(InVisibility=Collapsed)
   UnbindListenEvent() {is=title K2Node_CustomEvent}
   n2 = Get EnterExitBugBiomeEvent() -> (EnterExitBugBiomeEvent:Async Action Start Listening To Stateful Event Object Reference)
   > Is Valid(InputObject=@n2)
   > StopListeningToStatefulEvent(self:Async Action Start Listening To Stateful Event Object Reference=@n2)
   > Set EnterExitBugBiomeEvent(EnterExitBugBiomeEvent:Async Action Start Listening To Stateful Event Object Reference) -> (Output_Get:Async Action Start Listening To Stateful Event Object Reference)
   n3 = OnViewTargetChanged!() -> (
    OutputDelegate
    then
    PlayerController:Fort Player Controller Object Reference
    OldViewTarget:Actor Object Reference
    NewViewTarget:Actor Object Reference
   ) {is=title K2Node_CustomEvent}
   > UnbindListenEvent() {is=title K2Node_CallFunction}
   > n4 = StartListeningToStatefulEvent!(
    execute:Exec
    Target:Gameplay Event Router Component Object Reference=GetGlobalEventRouter()
    EventType:Script Struct Object Reference=/TidalCrane_Swarm_Boss/Blueprints/STRUCT_Event_BugBiome_PlayerEnteredExited
    Context:Object Reference=@n3.NewViewTarget
    EventBubblingRule:EEventBubblingRule Enum
    None:%=Self-Reference()
   ) -> (
    then:Exec
    AsyncTaskProxy:Async Action Start Listening To Stateful Event Object Reference
    OnEventReceived:Exec
    OnSavedEventStateExists:Exec
    OnEventStateCleared:Exec
    Payload:STRUCT Event Bug Biome Player Entered Exited Structure
   ) {is=title K2Node_StartListeningToStatefulEventAsyncAction}
   > Set EnterExitBugBiomeEvent(EnterExitBugBiomeEvent=@n4)
   n5 = Break STRUCT Event Bug Biome Player Entered Exited!(STRUCT_Event_BugBiome_PlayerEnteredExited:"STRUCT Event Bug Biome Player Entered Exited Structure (by ref)"=@n4.Payload) -> (
    IsInZone_5_#26:bool=~
    Manager_6_#27:BP Tidal Crane Boss Spawn Manager Soft Object Reference
   ) {is=title K2Node_BreakStruct}
   Branch(execute=[@n4.OnEventReceived, @n4.OnSavedEventStateExists], Condition=@n5)
   n6 = Resolve Soft Reference!(Input:BP Tidal Crane Boss Spawn Manager Soft Object Reference=@n5.Manager_6_#27) -> (Output:BP Tidal Crane Boss Spawn Manager Object Reference)
   > Set IsThreatHigh(IsThreatHigh:bool=IsThreatHigh!(self:BP Tidal Crane Boss Spawn Manager Object Reference=@n6) -> (IsHigh:bool)) -> (Output_Get:bool)
   n7 = Get IsThreatHigh() -> (IsThreatHigh:bool)
   n8 = Get ThreatHighText() -> (ThreatHighText:Text)
   n9 = Get ThreatLowText() -> (ThreatLowText:Text)
   > "SetText (Text)"(
    self=Get ThreatLevelDescriptionText() -> (ThreatLevelDescriptionText:Common Text Object Reference)
    InText=Format Text!(
     Format=Get ThreatFormatText() -> (ThreatFormatText:Text)
     threat:Text=Select(Option 0:Text=@n9, Option 1:Text=@n8, Index=@n7) -> (ReturnValue:Text)
    ) -> (Result)
   )
   > SetPoiNameText!(
    execute
    self
    Location:"Gameplay Tag Structure (by ref)"=GetPOITag!(self:BP Tidal Crane Boss Spawn Manager Object Reference=@n6) -> (POITag:Gameplay Tag Structure)
   ) -> (then)
   > SetVisibility(InVisibility=SelfHitTestInvisible)
   Event Destruct()
   > StopListeningToStatefulEvent(self=Get EnterExitBugBiomeEvent())
   > n10 = Cast To FortPlayerController(Object=GetOwningPlayer())
   n11 = Create Event()
   > Unbind Event from On View Target Changed(self:Fort Player Controller Object Reference=@n10, Delegate=@n11) {is=title K2Node_RemoveDelegate}
   Event OnInitialized()
   > BindToAnimationFinished(Animation=Get Anim_Outro() -> (Anim_Outro:Widget Animation Object Reference), Delegate=@n1)
   > n12 = GetDynamicMaterial(self=Get Image_Icon() -> (Image_Icon:Image Object Reference))
   > Set Mat_Icon(Mat_Icon:Material Instance Dynamic Object Reference=@n12) -> (Output_Get:Material Instance Dynamic Object Reference)
   > n13 = GetDynamicMaterial(self=Get Image_ThreatBacker() -> (Image_ThreatBacker:Image Object Reference))
   > Set Mat_ThreatBacker(Mat_ThreatBacker:Material Instance Dynamic Object Reference=@n13) -> (Output_Get:Material Instance Dynamic Object Reference)
   Event On Become Active() {is=title K2Node_Event}
   > PlayEntranceAnimation()
   > Set Timer by Event(Delegate=Create Event(), Time=Get DurationActive() -> (DurationActive:float), bLooping=false)
   Event PreConstruct()
   > "SetText (Text)"(self=Get Text_POIName() -> (Text_POIName:Common Text Object Reference), InText=Get POIFormatText() -> (POIFormatText:Text))
   Event Construct()
   > Parent: Construct()
   > SetVisibility(InVisibility=Collapsed)
   > n14 = Cast To FortPlayerController(Object=GetOwningPlayer())
   > Assign On View Target Changed!(execute:Exec, self:Fort Player Controller Object Reference=@n14, Delegate:"Delegate (by ref)"=@n11) -> (then:Exec) {is=title K2Node_AssignDelegate}
   > OnViewTargetChanged!(
    execute
    self
    PlayerController:Fort Player Controller Object Reference
    OldViewTarget:Actor Object Reference
    NewViewTarget:Actor Object Reference=GetViewTarget!(self:Controller Object Reference=@n14) -> (ReturnValue:Actor Object Reference)
   ) -> (then) {is=title K2Node_CallFunction}
   > Set DiscoverabilityComponent(DiscoverabilityComponent:Fort Controller Component Map Discoverability Object Reference=GetComponentByClass!(
     self:Actor Object Reference=GetOwningPlayer()
     ComponentClass:Actor Component Class Reference=/Script/FortniteGame.FortControllerComponent_MapDiscoverability
    ) -> (ReturnValue:Fort Controller Component Map Discoverability Object Reference)) -> (Output_Get:Fort Controller Component Map Discoverability Object Reference)
  ]}
 HidePOIMessage {nodes [
   HidePOIMessage() {is=title K2Node_FunctionEntry}
   > PlayAnimation(InAnimation=Get Anim_Outro())
  ]}
 PlayEntranceAnimation {nodes [
   PlayEntranceAnimation() {is=title K2Node_FunctionEntry}
   > n1 = Branch(Condition=Get IsThreatHigh())
   > PlayAnimation(InAnimation=Get Anim_Intro_HighThreat() -> (Anim_Intro_HighThreat:Widget Animation Object Reference))
   > PlaySound2D(Sound=/TheHivePOIUI/Assets/Sound/MSS_UI_Enter_TheHive_Popup)
   PlayAnimation(execute=@n1.else, InAnimation=Get Anim_Intro_LowThreat() -> (Anim_Intro_LowThreat:Widget Animation Object Reference))
   > PlaySound2D(Sound=&MSS_UI_Enter_TheHive_Popup)
  ]}
 SetPoiNameText {nodes [
   n1 = SetPoiNameText!() -> (then, Location:"Gameplay Tag Structure (by ref)") {is=title K2Node_FunctionEntry}
   n2 = Get DiscoverabilityComponent() -> (DiscoverabilityComponent:Fort Controller Component Map Discoverability Object Reference)
   > "SetText (Text)"(
    self=Get Text_POIName()
    InText=Format Text!(
     Format=Get POIFormatText()
     POI Name:"Text (by ref)"=GetMapTextForLocation!(self:Fort Controller Component Map Discoverability Object Reference=@n2, Location:"Gameplay Tag Structure (by ref)"=@n1) -> (ReturnValue:"Text (by ref)")
    ) -> (Result)
   )
  ]}
]
variables [
 DiscoverabilityComponent {type=FortControllerComponent_MapDiscoverability, default=null}
 DurationActive {type=number, default=4}
 EnterExitBugBiomeEvent {type=AsyncAction_StartListeningToStatefulEvent, default=null}
 IsThreatHigh {type=boolean}
 Mat_Icon {type=MaterialInstanceDynamic}
 Mat_ThreatBacker {type=MaterialInstanceDynamic}
 POIFormatText {type=string, default="Entering {POI Name}"}
 ThreatFormatText {type=string, default="Threat Level:  {threat}"}
 ThreatHighText {type=string, default=High}
 ThreatLowText {type=string, default=Low}
]
widgets {
 info {inheritedWidgetCount=1, parentClass=/Script/FortniteUI.AthenaCommonPrioritizedWidget, rootWidgetClass=&Overlay, widgetCount=17}
 widgets [
  Overlay_30 {
   widgetClassPath=&Overlay
   widget=Overlay_30
   properties {renderTransform {angle=0, scale=<1 1>, shear=<0 0>, translation=<0 -50>}}
  }
  VerticalBox_129 {
   widgetClassPath=&VerticalBox
   parent=Overlay_30
   widget=VerticalBox_129
   slot {class=OverlaySlot, properties {horizontalAlignment=HAlign_Center, padding=<ltrb 0 0 0 -90>}}
  }
  Overlay_0 {
   widgetClassPath=&Overlay
   parent=VerticalBox_129
   widget=Overlay_0
   slot {class=VerticalBoxSlot, properties {horizontalAlignment=HAlign_Left, verticalAlignment=VAlign_Top}}
  }
  Image_Backer {
   widgetClassPath=&Image
   bIsVariable=true
   parent=Overlay_0
   widget=Image_Backer
   properties {
    brush {
     bIsDynamicallyLoaded=false
     drawAs=Image
     imageType=FullColor
     mirroring=NoMirror
     resourceName=None
     resourceObject=/TheHivePOIUI/UI/Materials/M_UI_HivePOIEnter_Backer
     tiling=NoTile
     imageSize=<700 250>
     margin=<ltrb 0.4 0.4 0.4 0.4>
     outlineSettings {
      bUseBrushTransparency=false
      roundingType=HalfHeightRadius
      width=0
      color {colorUseRule=UseColor_Specified, specifiedColor=<rgba 0 0 0 0>}
      cornerRadii=<0 0 0 0>
     }
     tintColor {colorUseRule=UseColor_Specified, specifiedColor=<rgba 1 1 1 1>}
     uVRegion {bIsValid=false, max=<0 0>, min=<0 0>}
    }
    renderTransform {angle=0, scale=<1 1>, shear=<0 0>, translation=<0 3>}
   }
   slot {class=OverlaySlot, properties {horizontalAlignment=HAlign_Fill, verticalAlignment=VAlign_Fill}}
  }
  Image_Icon {
   widgetClassPath=&Image
   bIsVariable=true
   parent=Overlay_0
   widget=Image_Icon
   properties {brush {
     bIsDynamicallyLoaded=false
     drawAs=Image
     imageType=FullColor
     mirroring=NoMirror
     resourceName=None
     resourceObject=/TheHivePOIUI/UI/Materials/M_UI_HivePOIEnter_Icon
     tiling=NoTile
     imageSize=<600 250>
     margin=<ltrb 0 0 0 0>
     outlineSettings {
      bUseBrushTransparency=false
      roundingType=HalfHeightRadius
      width=0
      color {colorUseRule=UseColor_Specified, specifiedColor=<rgba 0 0 0 0>}
      cornerRadii=<0 0 0 0>
     }
     tintColor {colorUseRule=UseColor_Specified, specifiedColor=<rgba 1 1 1 1>}
     uVRegion {bIsValid=false, max=<0 0>, min=<0 0>}
    }}
   slot {class=OverlaySlot, properties {horizontalAlignment=HAlign_Center}}
  }
  Image_Glow {
   widgetClassPath=&Image
   bIsVariable=true
   parent=Overlay_0
   widget=Image_Glow
   properties {
    renderOpacity=0.05
    brush {
     bIsDynamicallyLoaded=false
     drawAs=Image
     imageType=FullColor
     mirroring=NoMirror
     resourceName=None
     resourceObject=/TidalCraneSharedActivitiesUI/Materials/MI_UI_HardEdgedGlow
     tiling=NoTile
     imageSize=<150 149>
     margin=<ltrb 0 0 0 0>
     outlineSettings {
      bUseBrushTransparency=false
      roundingType=HalfHeightRadius
      width=0
      color {colorUseRule=UseColor_Specified, specifiedColor=<rgba 0 0 0 0>}
      cornerRadii=<0 0 0 0>
     }
     tintColor {colorUseRule=UseColor_Specified, specifiedColor=<rgba 1 1 1 1>}
     uVRegion {bIsValid=false, max=<0 0>, min=<0 0>}
    }
    renderTransform {angle=0, scale=<1 1>, shear=<0 0>, translation=<0 -30>}
   }
   slot {class=OverlaySlot, properties {horizontalAlignment=HAlign_Center, verticalAlignment=VAlign_Center}}
  }
  Overlay_1 {
   widgetClassPath=&Overlay
   parent=Overlay_0
   widget=Overlay_1
   properties {renderTransform {angle=0, scale=<1 1>, shear=<0 0>, translation=<0 6>}}
   slot {class=OverlaySlot, properties {horizontalAlignment=HAlign_Center, verticalAlignment=VAlign_Center}}
  }
  Spacer_65 {
   widgetClassPath=&Spacer
   parent=Overlay_1
   widget=Spacer_65
   properties {size=<600 1>}
   slot {class=OverlaySlot, properties {horizontalAlignment=HAlign_Fill, verticalAlignment=VAlign_Fill}}
  }
  SizeBox_0 {
   widgetClassPath=&SizeBox
   parent=Overlay_1
   widget=SizeBox_0
   properties {bOverride_WidthOverride=true, widthOverride=500}
   slot {class=OverlaySlot, properties {horizontalAlignment=HAlign_Center, verticalAlignment=VAlign_Center}}
  }
  Text_POIName {
   widgetClassPath=&CommonTextBlock
   bIsVariable=true
   parent=SizeBox_0
   widget=Text_POIName
   properties {
    justification=Center
    scrollStyle=&TextScrollStyle-Base_C
    text="Entering {POI Name}"
    textTransformPolicy=ToUpper
    font {
     bForceMonospaced=false
     bMaterialIsStencil=false
     fontMaterial=null
     fontObject=&BurbankSmall
     letterSpacing=0
     monospacedWidth=1
     size=24.75
     skewAmount=0
     typefaceFontName=Black
     outlineSettings {
      bApplyOutlineToDropShadows=false
      bMiteredCorners=false
      bSeparateFillAlpha=false
      outlineMaterial=&M-FontShadow
      outlineSize=2
      outlineColor=<rgba 0 0 0 0.8>
     }
    }
    renderTransform {angle=0, scale=<1 1>, shear=<0 0>, translation=<0 5>}
    shadowOffset=<0 0>
   }
   slot {class=SizeBoxSlot}
  }
  Image_TextFx {
   widgetClassPath=&Image
   bIsVariable=true
   parent=Overlay_1
   widget=Image_TextFx
   properties {
    renderOpacity=0
    brush {
     bIsDynamicallyLoaded=false
     drawAs=Image
     imageType=FullColor
     mirroring=NoMirror
     resourceName=None
     resourceObject=/TidalCraneSharedActivitiesUI/Materials/M_UI_TextBlink
     tiling=NoTile
     imageSize=<32 32>
     margin=<ltrb 0 0 0 0>
     outlineSettings {
      bUseBrushTransparency=false
      roundingType=HalfHeightRadius
      width=0
      color {colorUseRule=UseColor_Specified, specifiedColor=<rgba 0 0 0 0>}
      cornerRadii=<0 0 0 0>
     }
     tintColor {colorUseRule=UseColor_Specified, specifiedColor=<rgba 1 1 1 1>}
     uVRegion {bIsValid=false, max=<0 0>, min=<0 0>}
    }
    renderTransform {angle=0, scale=<1.3 1.6>, shear=<0 0>, translation=<0 4>}
   }
   slot {class=OverlaySlot, properties {horizontalAlignment=HAlign_Fill, verticalAlignment=VAlign_Fill}}
  }
  Overlay_2 {
   widgetClassPath=&Overlay
   parent=VerticalBox_129
   widget=Overlay_2
   properties {renderTransform {angle=0, scale=<1.2 1.2>, shear=<0 0>, translation=<0 -70>}}
   slot {class=VerticalBoxSlot, properties {horizontalAlignment=HAlign_Center, verticalAlignment=VAlign_Top}}
  }
  Image_ThreatBacker {
   widgetClassPath=&Image
   bIsVariable=true
   parent=Overlay_2
   widget=Image_ThreatBacker
   properties {brush {
     bIsDynamicallyLoaded=false
     drawAs=Image
     imageType=FullColor
     mirroring=NoMirror
     resourceName=None
     resourceObject=/TheHivePOIUI/UI/Materials/M_HivePOIEnter_ThreatBacker
     tiling=NoTile
     imageSize=<50 50>
     margin=<ltrb 0 0 0 0>
     outlineSettings {
      bUseBrushTransparency=false
      roundingType=HalfHeightRadius
      width=0
      color {colorUseRule=UseColor_Specified, specifiedColor=<rgba 0 0 0 0>}
      cornerRadii=<0 0 0 0>
     }
     tintColor {colorUseRule=UseColor_Specified, specifiedColor=<rgba 1 1 1 1>}
     uVRegion {bIsValid=false, max=<0 0>, min=<0 0>}
    }}
   slot {class=OverlaySlot, properties {horizontalAlignment=HAlign_Fill, padding=<ltrb -20 0 -20 0>}}
  }
  SizeBox_1 {
   widgetClassPath=&SizeBox
   parent=Overlay_2
   widget=SizeBox_1
   properties {bOverride_MaxDesiredWidth=true, maxDesiredWidth=225, widthOverride=225}
   slot {class=OverlaySlot, properties {horizontalAlignment=HAlign_Center, verticalAlignment=VAlign_Center}}
  }
  ThreatLevelDescriptionText {
   widgetClassPath=&CommonTextBlock
   bIsVariable=true
   parent=SizeBox_1
   widget=ThreatLevelDescriptionText
   properties {
    applyLineHeightToBottomLine=false
    justification=Center
    scrollStyle=&TextScrollStyle-Base_C
    text="Threat Level:  {threat} "
    textTransformPolicy=ToUpper
    colorAndOpacity {colorUseRule=UseColor_Specified, specifiedColor=<rgba 1 0.2232 0 1>}
    font {
     bForceMonospaced=false
     bMaterialIsStencil=false
     fontMaterial=null
     fontObject=&BurbankSmall
     letterSpacing=0
     monospacedWidth=1
     size=12
     skewAmount=0
     typefaceFontName=Black
     outlineSettings {
      bApplyOutlineToDropShadows=false
      bMiteredCorners=false
      bSeparateFillAlpha=false
      outlineMaterial=&M-FontShadow
      outlineSize=2
      outlineColor=<rgba 1 0.3095 0 0.502>
     }
    }
    renderTransform {angle=0, scale=<1 1>, shear=<0 0>, translation=<0 1>}
    shadowOffset=<0 0>
   }
   slot {class=SizeBoxSlot}
  }
  Image_HighThreatTextGlow {
   widgetClassPath=&Image
   bIsVariable=true
   parent=Overlay_2
   widget=Image_HighThreatTextGlow
   properties {
    renderOpacity=0
    brush {
     bIsDynamicallyLoaded=false
     drawAs=Image
     imageType=FullColor
     mirroring=NoMirror
     resourceName=None
     resourceObject=/TheHivePOIUI/UI/Materials/M_UI_HardEdgedGlow
     tiling=NoTile
     imageSize=<203.4312 34.2083>
     margin=<ltrb 0 0 0 0>
     outlineSettings {
      bUseBrushTransparency=false
      roundingType=HalfHeightRadius
      width=0
      color {colorUseRule=UseColor_Specified, specifiedColor=<rgba 0 0 0 0>}
      cornerRadii=<0 0 0 0>
     }
     tintColor {colorUseRule=UseColor_Specified, specifiedColor=<rgba 1 1 1 1>}
     uVRegion {bIsValid=false, max=<0 0>, min=<0 0>}
    }
    renderTransform {angle=90, scale=<0.15 5>, shear=<0 0>, translation=<71 0>}
   }
   slot {class=OverlaySlot, properties {horizontalAlignment=HAlign_Center, verticalAlignment=VAlign_Center}}
  }
  TutorialHighlight {widgetClassPath=&TutorialHighlight}
 ]
}
