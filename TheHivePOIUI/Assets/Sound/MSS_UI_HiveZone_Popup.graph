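# One-shot stereo MetaSound UI sound for the Hive Zone popup in the Hive POI UI, routed through the Fort_UI sound class.
# On play, a single stereo Wave Player plays the Hivezone_UI_PopUp wave straight to the left and right outputs and signals completion when it ends.
# A concurrency override allows only one instance at a time and sets a retrigger time to limit rapid repeats.
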MetaSoundSource /TheHivePOIUI/Assets/Sound/MSS_UI_HiveZone_Popup
inputs [UE.Source.OnPlay {type=Trigger}]
interfaces [UE.OutputFormat.Stereo:1.0, UE.Source.OneShot:1.0, UE.Source:1.1]
outputs [
 "UE.OutputFormat.Stereo.Audio:0" {type=Audio}
 "UE.OutputFormat.Stereo.Audio:1" {type=Audio}
 UE.Source.OneShot.OnFinished {type=Trigger}
]
pages [Default {
  input_defaults {UE.Source.OnPlay=null}
  nodes [WavePlayer_0 = UE.Wave Player:Stereo(Play=^UE.Source.OnPlay) {also {defaults {Wave Asset=&/Hivezone_UI_PopUp}, version [1,1]}}]
  links [
   {dst="@UE.Source.OneShot.OnFinished", src=WavePlayer_0.On Finished}
   {dst="@UE.OutputFormat.Stereo.Audio:0", src=WavePlayer_0.Out Left}
   {dst="@UE.OutputFormat.Stereo.Audio:1", src=WavePlayer_0.Out Right}
  ]
 }]
properties {
 assetImportData=&self:AssetImportData
 bOverrideConcurrency=true
 numChannels=2
 outputFormat=Stereo
 sampleRate=48000
 soundClassObject=&Fort_UI
 volume=0.8
 concurrencyOverrides {maxCount=1, retriggerTime=2}
}
