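# Hive POI HUD bar that counts down remaining threat before the Queen is summoned, then shows her health.
# Listens for bug-biome enter/exit events on the current view target, binds to that zone's boss spawn manager, and plays intro or outro based on zone and threat state.
# Boss spawn swaps title and bar style.
# Material parameters drive the bar, with a delayed, animated delta trail during the summon phase.
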WidgetBlueprint /TheHivePOIUI/UI/Widgets/WBP_HivePOI_ProgressBar
animations [
 AnimIntro {
  binding_count=5
  length_seconds=1.8834
  object=&self:AnimIntro
  sequence {bindings [
    Text_Title {
     id=#1
     display_name=CommonText_Title
     possessed_class=CommonTextBlock
     tracks [
      {
       type=MovieSceneFloatTrack
       display_name=Render Opacity
       sections [{type=MovieSceneFloatSection, channels [None {type=MovieSceneScriptingFloatChannel, keys [[16.9987,0],[24.3333,0],[24.9987,1]]}]}]
      }
      {
       type=MovieScene2DTransformTrack
       sections [{
         type=MovieScene2DTransformSection
         end_frame=25
         start_frame=21
         channels [
          Translation.X {type=MovieSceneScriptingFloatChannel, keys [[21.3327,-30],[25,0],[27.6667,0]]}
          Translation.Y {type=MovieSceneScriptingFloatChannel, const [0,1]}
          Angle {type=MovieSceneScriptingFloatChannel, const [0,1]}
          Scale.X {type=MovieSceneScriptingFloatChannel, const [1,1]}
          Scale.Y {type=MovieSceneScriptingFloatChannel, const [1,1]}
          Shear.X {type=MovieSceneScriptingFloatChannel, const [0,1]}
          Shear.Y {type=MovieSceneScriptingFloatChannel, const [0,1]}
         ]
        }]
      }
     ]
    }
    Image_Glow {
     id=#2
     possessed_class=Image
     tracks [
      {
       type=MovieSceneFloatTrack
       display_name=Render Opacity
       sections [{type=MovieSceneFloatSection, channels [None {type=MovieSceneScriptingFloatChannel, keys [[25,0],[37.6667,0.1]]}]}]
      }
      {
       type=MovieScene2DTransformTrack
       sections [{
         type=MovieScene2DTransformSection
         channels [
          Translation.X {type=MovieSceneScriptingFloatChannel, const [0,1]}
          Translation.Y {type=MovieSceneScriptingFloatChannel, const [-32,1]}
          Angle {type=MovieSceneScriptingFloatChannel, const [0,1]}
          Scale.X {type=MovieSceneScriptingFloatChannel, const [1,1]}
          Scale.Y {type=MovieSceneScriptingFloatChannel, const [1,1]}
          Shear.X {type=MovieSceneScriptingFloatChannel, const [0,1]}
          Shear.Y {type=MovieSceneScriptingFloatChannel, const [0,1]}
         ]
        }]
      }
     ]
    }
    Image_TextFX {
     id=#3
     possessed_class=Image
     tracks [
      {
       type=MovieSceneWidgetMaterialTrack
       sections [{
         type=MovieSceneParameterSection
         channels [
          Blinky X Movement {type=MovieSceneScriptingFloatChannel, keys [[22.6667,0.266],[25.6667,-0.29]]}
          Mask {type=MovieSceneScriptingFloatChannel, keys [[22.6667,0],[25.6667,1]]}
         ]
        }]
      }
      {
       type=MovieSceneFloatTrack
       display_name=Render Opacity
       sections [{
         type=MovieSceneFloatSection
         channels [None {type=MovieSceneScriptingFloatChannel, keys [[15.3333,0],[17,0.5],[22.6667,0.5],[25.3333,1],[27.6667,1],[28.6667,0]]}]
        }]
      }
     ]
    }
    Image_ProgressBar {
     id=#4
     possessed_class=Image
     tracks [
      {
       type=MovieSceneWidgetMaterialTrack
       sections [{
         type=MovieSceneParameterSection
         channels [
          Double Dots Opacity {type=MovieSceneScriptingFloatChannel, keys [[12.3333,0],[12.6667,1],[13.3333,0],[14.6667,1],[15.3333,0.8],[16.6667,0],[18.3333,1]]}
          Double Dots Size {type=MovieSceneScriptingFloatChannel, thinned_from=4, keys [[12.6667,1],[14.3333,1],[14.6667,0]]}
          Outer Accent X Movement {type=MovieSceneScriptingFloatChannel, keys [[3.3333,1],[7.3333,0]]}
          Outer Bar Flash {type=MovieSceneScriptingFloatChannel, keys [[0,1],[19.6667,0.5],[26.6667,0]]}
          Progress Bar Reveal {type=MovieSceneScriptingFloatChannel, keys [[6,0],[11.3333,1]]}
          Progress Bar Flash {type=MovieSceneScriptingFloatChannel, keys [[6,1],[11.3333,1],[21.3333,0]]}
          Outer Frame Opacity {type=MovieSceneScriptingFloatChannel, keys [[0,0],[17.3333,0],[18.3333,1],[19.6667,0.5]]}
          Outer Accents Opacity {type=MovieSceneScriptingFloatChannel, const [1,1]}
          Progress Opacity {type=MovieSceneScriptingFloatChannel, keys [[21.6667,0],[26.3333,1]]}
          Border Thickness {type=MovieSceneScriptingFloatChannel, keys [[18,1],[25,0]]}
         ]
        }]
      }
      {
       type=MovieSceneFloatTrack
       display_name=Render Opacity
       sections [{type=MovieSceneFloatSection, channels [None {type=MovieSceneScriptingFloatChannel, const [1,1]}]}]
      }
     ]
    }
    Image_BossGlow {
     id=#5
     possessed_class=Image
     tracks [
      {
       type=MovieSceneFloatTrack
       display_name=Render Opacity
       sections [{type=MovieSceneFloatSection, channels [None {type=MovieSceneScriptingFloatChannel, keys [[25,0],[37.6667,0.025]]}]}]
      }
      {
       type=MovieScene2DTransformTrack
       sections [{
         type=MovieScene2DTransformSection
         channels [
          Translation.X {type=MovieSceneScriptingFloatChannel, const [0,1]}
          Translation.Y {type=MovieSceneScriptingFloatChannel, const [0,1]}
          Angle {type=MovieSceneScriptingFloatChannel, const [0,1]}
          Scale.X {type=MovieSceneScriptingFloatChannel, const [1,1]}
          Scale.Y {type=MovieSceneScriptingFloatChannel, const [1.4,1]}
          Shear.X {type=MovieSceneScriptingFloatChannel, const [0,1]}
          Shear.Y {type=MovieSceneScriptingFloatChannel, const [0,1]}
         ]
        }]
      }
     ]
    }
   ]}
 }
 AnimOutro {
  binding_count=4
  length_seconds=1.15
  object=&self:AnimOutro
  sequence {bindings [
    Text_Title {
     id=#6
     display_name=CommonText_Title
     possessed_class=CommonTextBlock
     tracks [{
       type=MovieSceneFloatTrack
       display_name=Render Opacity
       sections [{type=MovieSceneFloatSection, channels [None {type=MovieSceneScriptingFloatChannel, keys [[0,1],[5,0]]}]}]
      }]
    }
    Image_BossGlow {
     id=#7
     possessed_class=Image
     tracks [
      {
       type=MovieSceneFloatTrack
       display_name=Render Opacity
       sections [{
         type=MovieSceneFloatSection
         end_frame=13
         start_frame=0
         channels [None {type=MovieSceneScriptingFloatChannel, keys [[0,0.025],[1,0],[13,0]]}]
        }]
      }
      {
       type=MovieScene2DTransformTrack
       sections [{
         type=MovieScene2DTransformSection
         channels [
          Translation.X {type=MovieSceneScriptingFloatChannel, const [0,1]}
          Translation.Y {type=MovieSceneScriptingFloatChannel, const [0,1]}
          Angle {type=MovieSceneScriptingFloatChannel, const [0,1]}
          Scale.X {type=MovieSceneScriptingFloatChannel, keys [[0,1],[5,0]]}
          Scale.Y {type=MovieSceneScriptingFloatChannel, const [1.4,1]}
          Shear.X {type=MovieSceneScriptingFloatChannel, const [0,1]}
          Shear.Y {type=MovieSceneScriptingFloatChannel, const [0,1]}
         ]
        }]
      }
     ]
    }
    Image_Glow {
     id=#8
     possessed_class=Image
     tracks [
      {
       type=MovieSceneFloatTrack
       display_name=Render Opacity
       sections [{type=MovieSceneFloatSection, end_frame=13, start_frame=0, channels [None {type=MovieSceneScriptingFloatChannel, keys [[0,1],[1,0],[13,0]]}]}]
      }
      {
       type=MovieScene2DTransformTrack
       sections [{
         type=MovieScene2DTransformSection
         channels [
          Translation.X {type=MovieSceneScriptingFloatChannel, const [0,1]}
          Translation.Y {type=MovieSceneScriptingFloatChannel, const [-32,1]}
          Angle {type=MovieSceneScriptingFloatChannel, const [0,1]}
          Scale.X {type=MovieSceneScriptingFloatChannel, keys [[0,1],[5,0]]}
          Scale.Y {type=MovieSceneScriptingFloatChannel, const [1,1]}
          Shear.X {type=MovieSceneScriptingFloatChannel, const [0,1]}
          Shear.Y {type=MovieSceneScriptingFloatChannel, const [0,1]}
         ]
        }]
      }
     ]
    }
    Image_ProgressBar {
     id=#9
     possessed_class=Image
     tracks [{
       type=MovieSceneWidgetMaterialTrack
       sections [{
         type=MovieSceneParameterSection
         channels [
          Double Dots Opacity {type=MovieSceneScriptingFloatChannel, keys [[0,1],[3,0],[3.3333,1],[4,0],[5.3333,1],[6,0.8],[7.3333,0],[9,0]]}
          Double Dots Size {type=MovieSceneScriptingFloatChannel, thinned_from=5, keys [[0,0],[3.3333,1],[5,1],[5.3333,0]]}
          Outer Accent X Movement {type=MovieSceneScriptingFloatChannel, keys [[0,0],[7.3333,0],[13,1],[15,1]]}
          Outer Bar Flash {type=MovieSceneScriptingFloatChannel, keys [[0,0],[7.3333,0],[13,1],[15,1],[23,0]]}
          Progress Bar Reveal {type=MovieSceneScriptingFloatChannel, keys [[0,1],[7,1],[9,0]]}
          Progress Bar Flash {type=MovieSceneScriptingFloatChannel, const [0,1]}
          Outer Frame Opacity {type=MovieSceneScriptingFloatChannel, keys [[0,1],[5,1],[7,0]]}
          Outer Accents Opacity {type=MovieSceneScriptingFloatChannel, thinned_from=4, keys [[0,1],[15,1],[23,0]]}
          Progress Opacity {type=MovieSceneScriptingFloatChannel, keys [[0,1],[3,0]]}
         ]
        }]
      }]
    }
   ]}
 }
 AnimBarSwap {
  binding_count=5
  length_seconds=1.4667
  object=&self:AnimBarSwap
  sequence {
   bindings [
    Text_Title {
     id=#10
     display_name=CommonText_Title
     possessed_class=CommonTextBlock
     tracks [
      {
       type=MovieScene2DTransformTrack
       sections [{
         type=MovieScene2DTransformSection
         channels [
          Translation.X {type=MovieSceneScriptingFloatChannel, const [0,1]}
          Translation.Y {type=MovieSceneScriptingFloatChannel, const [0,1]}
          Angle {type=MovieSceneScriptingFloatChannel, const [0,1]}
          Scale.X {type=MovieSceneScriptingFloatChannel, const [1,2]}
          Scale.Y {type=MovieSceneScriptingFloatChannel, const [1,1]}
          Shear.X {type=MovieSceneScriptingFloatChannel, const [0,1]}
          Shear.Y {type=MovieSceneScriptingFloatChannel, const [0,1]}
         ]
        }]
      }
      {
       type=MovieSceneFloatTrack
       display_name=Render Opacity
       sections [{type=MovieSceneFloatSection, channels [None {type=MovieSceneScriptingFloatChannel, keys [[72,0],[81,1]]}]}]
      }
     ]
    }
    Image_TextFX {
     id=#11
     possessed_class=Image
     tracks [
      {
       type=MovieSceneFloatTrack
       display_name=Render Opacity
       sections [{type=MovieSceneFloatSection, channels [None {type=MovieSceneScriptingFloatChannel, thinned_from=5, keys [[62,0],[70,1],[85,1],[88,0]]}]}]
      }
      {
       type=MovieSceneWidgetMaterialTrack
       sections [{
         type=MovieSceneParameterSection
         channels [
          Blinky X Movement {type=MovieSceneScriptingFloatChannel, keys [[70,0.296],[78,-0.3]]}
          Mask {type=MovieSceneScriptingFloatChannel, keys [[70,0],[79,1]]}
         ]
        }]
      }
     ]
    }
    Image_BossGlow {
     id=#12
     possessed_class=Image
     tracks [
      {
       type=MovieSceneFloatTrack
       display_name=Render Opacity
       sections [{type=MovieSceneFloatSection, channels [None {type=MovieSceneScriptingFloatChannel, keys [[57,1],[82,0.071]]}]}]
      }
      {
       type=MovieScene2DTransformTrack
       sections [{
         type=MovieScene2DTransformSection
         channels [
          Translation.X {type=MovieSceneScriptingFloatChannel, default=0}
          Translation.Y {type=MovieSceneScriptingFloatChannel, default=0}
          Angle {type=MovieSceneScriptingFloatChannel, default=0}
          Scale.X {type=MovieSceneScriptingFloatChannel, default=1, const [1,1]}
          Scale.Y {type=MovieSceneScriptingFloatChannel, keys [[57,0],[71,2]]}
          Shear.X {type=MovieSceneScriptingFloatChannel, default=0}
          Shear.Y {type=MovieSceneScriptingFloatChannel, default=0}
         ]
        }]
      }
     ]
    }
    Image_ProgressBar {
     id=#13
     possessed_class=Image
     tracks [{
       type=MovieSceneWidgetMaterialTrack
       sections [{
         type=MovieSceneParameterSection
         channels [
          Double Dots Opacity {type=MovieSceneScriptingFloatChannel, keys [[0,1],[15,0],[16,1],[18,0],[22,1],[24,0.8],[28,0],[33,1]]}
          Double Dots Size {type=MovieSceneScriptingFloatChannel, thinned_from=5, keys [[0,0],[16,1],[21,1],[22,0]]}
          Outer Accent X Movement {type=MovieSceneScriptingFloatChannel, const [0,1]}
          Outer Bar Flash {type=MovieSceneScriptingFloatChannel, keys [[0,0],[8,0.5],[23,0]]}
          Progress Bar Reveal {type=MovieSceneScriptingFloatChannel, const [1,1]}
          Progress Bar Flash {type=MovieSceneScriptingFloatChannel, thinned_from=5, keys [[0,0],[59,0],[62,1],[66,0]]}
          Outer Accents Opacity {type=MovieSceneScriptingFloatChannel, const [1,1]}
          Progress Opacity {type=MovieSceneScriptingFloatChannel, keys [[35,0],[41,1]]}
          BugBossProgressSwitch {type=MovieSceneScriptingFloatChannel, keys [[0,0],[8,1]]}
          Progress {type=MovieSceneScriptingFloatChannel, keys [[0,0],[35,0],[60,1.0022]]}
          Border Thickness {type=MovieSceneScriptingFloatChannel, keys [[0,0],[8,1],[37,1],[42,0]]}
          Border Collapse {type=MovieSceneScriptingFloatChannel, keys [[0,0],[8,1],[29,1],[37,0]]}
          Outer Frame Opacity {type=MovieSceneScriptingFloatChannel, keys [[0,1],[8,0],[29,0],[42,1]]}
         ]
        }]
      }]
    }
    Image_Glow {
     id=#14
     possessed_class=Image
     tracks [
      {
       type=MovieSceneWidgetMaterialTrack
       sections [{
         type=MovieSceneParameterSection
         channels [
          Progress {type=MovieSceneScriptingFloatChannel, const [0,1]}
          GradientBugBossColorSwitch {type=MovieSceneScriptingFloatChannel, const [0,1]}
         ]
        }]
      }
      {
       type=MovieScene2DTransformTrack
       sections [{
         type=MovieScene2DTransformSection
         channels [
          Translation.X {type=MovieSceneScriptingFloatChannel, const [0,1]}
          Translation.Y {type=MovieSceneScriptingFloatChannel, keys [[60.191,-27.9796],[65.808,-32]]}
          Angle {type=MovieSceneScriptingFloatChannel, const [0,1]}
          Scale.X {type=MovieSceneScriptingFloatChannel, const [1,2]}
          Scale.Y {type=MovieSceneScriptingFloatChannel, keys [[60.191,0],[69,1]]}
          Shear.X {type=MovieSceneScriptingFloatChannel, const [0,1]}
          Shear.Y {type=MovieSceneScriptingFloatChannel, const [0,1]}
         ]
        }]
      }
      {
       type=MovieSceneFloatTrack
       display_name=Render Opacity
       sections [{type=MovieSceneFloatSection, channels [None {type=MovieSceneScriptingFloatChannel, const [0,1]}]}]
      }
     ]
    }
   ]
   display_rate {numerator=60}
  }
 }
 AnimDeltaChange {
  binding_count=1
  length_seconds=0.3
  object=&self:AnimDeltaChange
  sequence {bindings [WBP_HivePOI_ProgressBar_C_10 {
     id=#15
     possessed_class=WBP_HivePOI_ProgressBar_C
     tracks [{
       type=MovieSceneDoubleTrack
       display_name=Delta Transition Progress
       sections [{type=MovieSceneDoubleSection, end_frame=6, start_frame=0, channels [None {type=MovieSceneScriptingDoubleChannel, keys [[0,0],[6,1]]}]}]
      }]
    }]}
 }
]
graphs [
 BindToBossManagerEvents {nodes [
   n1 = BindToBossManagerEvents!() -> (then, NewManager:BP Tidal Crane Boss Spawn Manager Object Reference) {is=title K2Node_FunctionEntry}
   > n2 = Delta Gate Wildcard!(
    execute:Exec
    Variable:"BP Tidal Crane Boss Spawn Manager Object Reference (by ref)"=Get BossManager() -> (BossManager:BP Tidal Crane Boss Spawn Manager Object Reference)
    New Value:BP Tidal Crane Boss Spawn Manager Object Reference=@n1
   ) -> (
    Success:Exec
    Updated Variable:"BP Tidal Crane Boss Spawn Manager Object Reference (by ref)"
    Previous Variable:BP Tidal Crane Boss Spawn Manager Object Reference
   ) {is=title K2Node_MacroInstance}
   > Set PreviousManager(PreviousManager:BP Tidal Crane Boss Spawn Manager Object Reference=@n2.Previous Variable) -> (Output_Get:BP Tidal Crane Boss Spawn Manager Object Reference)
   > n3 = Sequence()
   > Is Valid(InputObject=Get PreviousManager() -> (PreviousManager:BP Tidal Crane Boss Spawn Manager Object Reference))
   n4 = Create Event()
   > Unbind Event from On Boss Set(self:BP Tidal Crane Boss Spawn Manager Object Reference=Get PreviousManager(), Delegate=@n4) {is=title K2Node_RemoveDelegate}
   n5 = Create Event()
   > Unbind Event from On Boss Summon Progress Made(self:BP Tidal Crane Boss Spawn Manager Object Reference=Get PreviousManager(), Delegate=@n5) {is=title K2Node_RemoveDelegate}
   n6 = Create Event()
   > Unbind Event from On Event Finished(self:BP Tidal Crane Boss Spawn Manager Object Reference=Get PreviousManager(), Delegate=@n6) {is=title K2Node_RemoveDelegate}
   Is Valid(exec=@n3.then_1, InputObject=Get BossManager())
   > Bind Event to On Boss Set(self:BP Tidal Crane Boss Spawn Manager Object Reference=Get BossManager(), Delegate=@n4) {is=title K2Node_AddDelegate}
   > Bind Event to On Boss Summon Progress Made(self:BP Tidal Crane Boss Spawn Manager Object Reference=Get BossManager(), Delegate=@n5) {is=title K2Node_AddDelegate}
   > Bind Event to On Event Finished(self:BP Tidal Crane Boss Spawn Manager Object Reference=Get BossManager(), Delegate=@n6) {is=title K2Node_AddDelegate}
   > SetBoss!(
    execute
    self
    Boss:Building Gameplay Actor Object Reference=Get TidalCraneBoss!(self:BP Tidal Crane Boss Spawn Manager Object Reference=Get BossManager()) -> (TidalCraneBoss:Building Gameplay Actor Object Reference)
   ) -> (then)
   > InitalizeProgressBar!(execute, self, AnimateSwap:bool=false) -> (then)
  ]}
 ClearBindings {nodes [
   ClearBindings() {is=title K2Node_FunctionEntry}
   > StopListeningToStatefulEvent(self:Async Action Start Listening To Stateful Event Object Reference=Get EnterExitBugBiomeEvent() -> (EnterExitBugBiomeEvent:Async Action Start Listening To Stateful Event Object Reference))
   > BindToBossManagerEvents!(execute, self, NewManager:BP Tidal Crane Boss Spawn Manager Object Reference) -> (then) {is=title K2Node_CallFunction}
   > SetBoss()
   > n1 = Cast To FortPlayerController(Object=GetOwningPlayer())
   > Unbind Event from On View Target Changed(self:Fort Player Controller Object Reference=@n1, Delegate=Create Event()) {is=title K2Node_RemoveDelegate}
  ]}
 EventGraph {nodes [
   PlayIntroAnimation() {is=title K2Node_CustomEvent}
   > StopAnimation(InAnimation=Get AnimOutro() -> (AnimOutro:Widget Animation Object Reference))
   > Play Animation with Finished event!(
    execute:Exec
    Widget:User Widget Object Reference=Self-Reference()
    InAnimation:Widget Animation Object Reference=Get AnimIntro() -> (AnimIntro:Widget Animation Object Reference)
    StartAtTime:float=0
    NumLoopsToPlay:int=1
    PlayMode:EUMGSequencePlayMode Enum
    PlaybackSpeed:float=1
   ) -> (then:Exec, Result:Widget Animation Handle Structure, Finished:Exec) {is=title K2Node_PlayAnimation2}
   UnbindListenEvent() {is=title K2Node_CustomEvent}
   n1 = Get EnterExitBugBiomeEvent()
   > Is Valid(InputObject=@n1)
   > StopListeningToStatefulEvent(self=@n1)
   > Set EnterExitBugBiomeEvent(EnterExitBugBiomeEvent:Async Action Start Listening To Stateful Event Object Reference) -> (Output_Get:Async Action Start Listening To Stateful Event Object Reference)
   n2 = OnViewTargetChanged!() -> (
    OutputDelegate
    then
    PlayerController:Fort Player Controller Object Reference
    OldViewTarget:Actor Object Reference
    NewViewTarget:Actor Object Reference
   ) {is=title K2Node_CustomEvent}
   > UnbindListenEvent() {is=title K2Node_CallFunction}
   > n3 = StartListeningToStatefulEvent!(
    execute:Exec
    Target:Gameplay Event Router Component Object Reference=GetGlobalEventRouter()
    EventType:Script Struct Object Reference=/TidalCrane_Swarm_Boss/Blueprints/STRUCT_Event_BugBiome_PlayerEnteredExited
    Context:Object Reference=@n2.NewViewTarget
    EventBubblingRule:EEventBubblingRule Enum
    None:%=Self-Reference()
   ) -> (
    then:Exec
    AsyncTaskProxy:Async Action Start Listening To Stateful Event Object Reference
    OnEventReceived:Exec
    OnSavedEventStateExists:Exec
    OnEventStateCleared:Exec
    Payload:STRUCT Event Bug Biome Player Entered Exited Structure
   ) {is=title K2Node_StartListeningToStatefulEventAsyncAction}
   > Set EnterExitBugBiomeEvent(EnterExitBugBiomeEvent=@n3)
   n4 = Break STRUCT Event Bug Biome Player Entered Exited!(STRUCT_Event_BugBiome_PlayerEnteredExited:"STRUCT Event Bug Biome Player Entered Exited Structure (by ref)"=@n3.Payload) -> (
    IsInZone_5_#16:bool=~
    Manager_6_#17:BP Tidal Crane Boss Spawn Manager Soft Object Reference
   ) {is=title K2Node_BreakStruct}
   n5 = Set IsInZone(execute=[@n3.OnEventReceived, @n3.OnSavedEventStateExists], IsInZone:bool=@n4) -> (Output_Get:bool)
   > n6 = Branch(Condition=@n5)
   > BindToBossManagerEvents(NewManager=Resolve Soft Reference!(Input:BP Tidal Crane Boss Spawn Manager Soft Object Reference=@n4.Manager_6_#17) -> (Output:BP Tidal Crane Boss Spawn Manager Object Reference))
   > n7 = SetVisibility(InVisibility=SelfHitTestInvisible)
   PlayEnterExit!(execute=[@n7.then, @n6.else], self, IsEntering:bool=Get IsInZone() -> (IsInZone:bool)) -> (then)
   PlayOutroAnimation() {is=title K2Node_CustomEvent}
   > StopAnimation(InAnimation=Get AnimIntro())
   > Set IsHiveUIShowing(IsHiveUIShowing:bool=false) -> (Output_Get:bool)
   > n8 = Play Animation with Finished event(Widget=Self-Reference(), InAnimation=Get AnimOutro(), StartAtTime=0, NumLoopsToPlay=1, PlaybackSpeed=1)
   n9 = Branch(execute=@n8.Finished, Condition=Get IsHiveUIShowing() -> (IsHiveUIShowing:bool))
   SetVisibility(execute=@n9.else, InVisibility=Collapsed)
   Event On Become Active() {is=title K2Node_Event}
   > Set IsActive(IsActive:bool=true) -> (Output_Get:bool)
   > PlayEnterExit(IsEntering=Get IsInZone())
   Event Construct()
   > n10 = GetDynamicMaterial(self=Get Image_ProgressBar() -> (Image_ProgressBar:Image Object Reference))
   > Set Mat_ProgressBar(Mat_ProgressBar:Material Instance Dynamic Object Reference=@n10) -> (Output_Get:Material Instance Dynamic Object Reference)
   > SetVisibility(InVisibility=Collapsed)
   > SetTouchLayout()
   > n11 = Cast To FortPlayerController(Object=GetOwningPlayer())
   > Bind Event to On View Target Changed(self:Fort Player Controller Object Reference=@n11, Delegate=Create Event()) {is=title K2Node_AddDelegate}
   > OnViewTargetChanged!(
    execute
    self
    PlayerController:Fort Player Controller Object Reference
    OldViewTarget:Actor Object Reference
    NewViewTarget:Actor Object Reference=GetViewTarget!(self:Controller Object Reference=@n11) -> (ReturnValue:Actor Object Reference)
   ) -> (then) {is=title K2Node_CallFunction}
   Event On Become Inactive() {is=title K2Node_Event}
   > Set IsActive(IsActive=false)
   Event Destruct()
   > ClearBindings() {is=title K2Node_CallFunction}
  ]}
 InitalizeProgressBar {nodes [
   InitalizeProgressBar!() -> (then, AnimateSwap:bool=~) {is=title K2Node_FunctionEntry}
   > n1 = Branch(Condition=IsBossSummonProgressComplete!() -> (isQueenActive:bool))
   > "SetText (Text)"(self=Get Text_Title() -> (Text_Title:Common Text Object Reference), InText=Get Text_Queen() -> (Text_Queen:Text))
   > StopAnimation(InAnimation=Get AnimDeltaChange() -> (AnimDeltaChange:Widget Animation Object Reference))
   > SetScalarParameterValue(
    self=Get Mat_ProgressBar() -> (Mat_ProgressBar:Material Instance Dynamic Object Reference)
    ParameterName=Get Param_DeltaProgress() -> (Param_DeltaProgress:Name=None)
    Value=0
   )
   > Clear and Invalidate Timer by Handle(Handle=Get DelayForDeltaTimerHandle() -> (DelayForDeltaTimerHandle:Timer Handle Structure))
   > OnBossHealthChanged()
   > SetVisibility(self=Get Image_BossGlow() -> (Image_BossGlow:Image Object Reference), InVisibility=SelfHitTestInvisible)
   > n2 = Branch(Condition=Get AnimateSwap() -> (AnimateSwap:bool))
   > PlayAnimation(InAnimation=Get AnimBarSwap() -> (AnimBarSwap:Widget Animation Object Reference))
   SetVisibility(execute=@n2.else, self=Get Image_Glow() -> (Image_Glow:Image Object Reference), InVisibility=Collapsed)
   > SetScalarParameterValue(self=Get Mat_ProgressBar(), ParameterName=BugBossProgressSwitch, Value=1)
   SetScalarParameterValue(execute=@n1.else, self=Get Mat_ProgressBar(), ParameterName=BugBossProgressSwitch, Value=0)
   > "SetText (Text)"(self=Get Text_Title(), InText=Get Text_Nests() -> (Text_Nests:Text))
   n3 = Get BossManager()
   > SetScalarParameterValue(
    self=Get Mat_ProgressBar()
    ParameterName=Get Param_DeltaProgress() -> (Param_DeltaProgress=None)
    Value=float - float(A=1, B=Get BossSummonProgress!(self:BP Tidal Crane Boss Spawn Manager Object Reference=@n3) -> (BossSummonProgress:double))
   )
   > OnBossSummonProgressMade!(execute, self, New Progress:double=Get BossSummonProgress(self=Get BossManager())) -> (then)
   > SetVisibility(self=Get Image_BossGlow(), InVisibility=Collapsed)
   > SetVisibility(self=Get Image_Glow(), InVisibility=SelfHitTestInvisible)
  ]}
 IsBossSummonProgressComplete {nodes [
   IsBossSummonProgressComplete() {is=title K2Node_FunctionEntry}
   n1 = Get BossManager()
   > Return Node!(execute, isQueenActive:bool="float >= float"(A=Get BossSummonProgress(self=@n1), B=1))
  ]}
 OnBossHealthChanged {nodes [
   OnBossHealthChanged() {is=title K2Node_FunctionEntry}
   > SetProgress!(
    execute
    self
    Value:float=GetHealthPercent!(self:Building Actor Object Reference=Get TidalCraneBoss!() -> (TidalCraneBoss:BGA Tidal Crane Boss Object Reference)) -> (ReturnValue:float)
   ) -> (then)
  ]}
 OnBossSet {nodes [
   n1 = OnBossSet!() -> (then, Boss:Building Gameplay Actor Object Reference) {is=title K2Node_FunctionEntry}
   > SetBoss(Boss=@n1)
   > InitalizeProgressBar(AnimateSwap=true) {is=title K2Node_CallFunction}
  ]}
 OnBossSummonProgressMade {nodes [
   n1 = OnBossSummonProgressMade!() -> (then, New Progress:double=~) {is=title K2Node_FunctionEntry}
   > n2 = Branch(Condition=IsBossSummonProgressComplete() {is=title K2Node_CallFunction})
   SetProgress(execute=@n2.else, Value=float - float(A=1, B=@n1))
  ]}
 OnDeltaTimerFinished {nodes [
   OnDeltaTimerFinished() {is=title K2Node_FunctionEntry}
   > n1 = Get Scalar Parameter Value!(
    execute
    self:Material Instance Dynamic Object Reference=Get Mat_ProgressBar()
    ParameterName:Name=Get Param_DeltaProgress() -> (Param_DeltaProgress=None)
   ) -> (then, ReturnValue:float) {is=title K2Node_CallFunction}
   > Set StartingDeltaProgress(StartingDeltaProgress:double=@n1) -> (Output_Get:double)
   > PlayAnimation(InAnimation=Get AnimDeltaChange())
  ]}
 OnEventFinished {nodes [
   OnEventFinished() {is=title K2Node_FunctionEntry}
   > PlayOutroAnimation() {is=title K2Node_CallFunction}
  ]}
 PlayEnterExit {nodes [
   n1 = PlayEnterExit!() -> (then, IsEntering:bool=~) {is=title K2Node_FunctionEntry}
   > Branch(Condition=Get IsActive() -> (IsActive:bool))
   n2 = Get BossManager()
   > n3 = Delta Gate Wildcard(
    Variable:"Boolean (by ref)"=Get IsHiveUIShowing()
    New Value:bool=AND Boolean(A=@n1, B=IsThreatHigh!(self:BP Tidal Crane Boss Spawn Manager Object Reference=@n2) -> (IsHigh:bool))
   ) -> (Updated Variable:"Boolean (by ref)", Previous Variable:bool=~)
   > n4 = Branch(Condition=@n3)
   > PlayIntroAnimation() {is=title K2Node_CallFunction}
   > InitalizeProgressBar(AnimateSwap=false)
   > PlaySound2D(Sound=/TheHivePOIUI/Assets/Sound/MSS_UI_HiveZone_Popup)
   PlayOutroAnimation(execute=@n4.else)
  ]}
 SetBoss {nodes [
   n1 = SetBoss!() -> (then, Boss:Building Gameplay Actor Object Reference) {is=title K2Node_FunctionEntry}
   > n2 = Delta Gate Wildcard(Variable:"Building Gameplay Actor Object Reference (by ref)"=Get TidalCraneBoss(), New Value:Building Gameplay Actor Object Reference=@n1) -> (Updated Variable:"Building Gameplay Actor Object Reference (by ref)", Previous Variable:Building Gameplay Actor Object Reference)
   > Set PreviousVariable(PreviousVariable:Building Gameplay Actor Object Reference=@n2.Previous Variable) -> (Output_Get:Building Gameplay Actor Object Reference)
   > n3 = Sequence()
   > Is Valid(InputObject=Get PreviousVariable() -> (PreviousVariable:Building Gameplay Actor Object Reference))
   n4 = Create Event()
   > Unbind Event from On Building Health Changed(self:Building Actor Object Reference=Get PreviousVariable(), Delegate=@n4) {is=title K2Node_RemoveDelegate}
   Is Valid(exec=@n3.then_1, InputObject=Get TidalCraneBoss())
   > Bind Event to On Building Health Changed(self:Building Actor Object Reference=Get TidalCraneBoss(), Delegate=@n4) {is=title K2Node_AddDelegate}
  ]}
 SetDeltaTransitionProgress {nodes [
   n1 = SetDeltaTransitionProgress!() -> (then, Progress:double=~) {is=title K2Node_FunctionEntry}
   > n2 = Get!(execute:Exec) -> (then:Exec, else:Exec, Mat_ProgressBar:Material Instance Dynamic Object Reference)
   > n3 = Get Scalar Parameter Value(self=@n2, ParameterName=Progress)
   > SetScalarParameterValue(
    self=Get Mat_ProgressBar()
    ParameterName=Get Param_DeltaProgress() -> (Param_DeltaProgress=None)
    Value=Lerp(A=Get StartingDeltaProgress() -> (StartingDeltaProgress:double), B=@n3, Alpha=@n1)
   )
  ]}
 SetProgress {nodes [
   n1 = SetProgress!() -> (then, Value:float=~) {is=title K2Node_FunctionEntry}
   > SetScalarParameterValue(self=Get Mat_ProgressBar(), ParameterName=Progress, Value=@n1)
   > n2 = GetDynamicMaterial(self=Get Image_Glow())
   > SetScalarParameterValue(self=@n2, ParameterName=Progress, Value=Get Value() -> (Value:float))
   > n3 = Get Scalar Parameter Value(self=Get Mat_ProgressBar(), ParameterName=Get Param_DeltaProgress() -> (Param_DeltaProgress=None))
   n4 = "Nearly Equal (Float)"(A=@n3, B=Get Value(), ErrorTolerance=0)
   n5 = IsBossSummonProgressComplete()
   > Branch(Condition=AND Boolean(A=NOT Boolean(A=@n5), B=NOT Boolean(A=@n4)))
   > n6 = Set Timer by Event(Delegate=Create Event(), Time=Get DelayForDelta() -> (DelayForDelta:double), bLooping=false)
   > Set DelayForDeltaTimerHandle(DelayForDeltaTimerHandle:Timer Handle Structure=@n6) -> (Output_Get:Timer Handle Structure)
   > StopAnimation(InAnimation=Get AnimDeltaChange())
  ]}
 SetTouchLayout {nodes [
   SetTouchLayout() {is=title K2Node_FunctionEntry}
   > Branch(Condition=IsUsingTouch(PlayerController=GetOwningPlayer()))
   > n1 = Branch(Condition=IsTabletAspectRatio!() -> (ReturnValue:bool))
   > SetWidthOverride(self=Get SizeBox_All() -> (SizeBox_All:Size Box Object Reference), InWidthOverride=450)
   SetWidthOverride(execute=@n1.else, self=Get SizeBox_All(), InWidthOverride=600)
   SetPadding!(
    execute
    self:Size Box Slot Object Reference=SlotAsSizeBoxSlot!(Widget:Widget Object Reference=Get VerticalBox_All() -> (VerticalBox_All:Vertical Box Object Reference)) -> (ReturnValue:Size Box Slot Object Reference)
    InPadding_Left:float=0
    InPadding_Top:float=-32
    InPadding_Right:float=0
    InPadding_Bottom:float=0
   ) -> (then)
  ]}
]
variables [
 BossManager {type=BP_TidalCrane_BossSpawnManager_C, default=null}
 DelayForDelta {type=number, default=0.75}
 DelayForDeltaTimerHandle {type=TimerHandle}
 DeltaTransitionProgress {type=number}
 EnterExitBugBiomeEvent {type=AsyncAction_StartListeningToStatefulEvent, default=null}
 IsActive {type=boolean}
 IsHiveUIShowing {type=boolean}
 IsInZone {type=boolean}
 Mat_ProgressBar {type=MaterialInstanceDynamic}
 Param_DeltaProgress {type=string, default=DeltaProgress}
 StartingDeltaProgress {type=number}
 Text_Nests {type=string, default="Neutralize the threat!"}
 Text_Queen {type=string, default="Slay the Queen!"}
 TidalCraneBoss {type=BGA_TidalCrane_Boss_C, default=null}
]
widgets {
 info {inheritedWidgetCount=1, parentClass=/Script/FortniteUI.AthenaCommonPrioritizedWidget, rootWidgetClass=&SizeBox, widgetCount=10}
 widgets [
  SizeBox_All {widgetClassPath=&SizeBox, bIsVariable=true, widget=SizeBox_All, properties {bOverride_WidthOverride=true, widthOverride=700}}
  VerticalBox_All {widgetClassPath=&VerticalBox, bIsVariable=true, parent=SizeBox_All, widget=VerticalBox_All, slot {class=SizeBoxSlot}}
  Overlay_ProgressBar {widgetClassPath=&Overlay, bIsVariable=true, parent=VerticalBox_All, widget=Overlay_ProgressBar, slot {class=VerticalBoxSlot}}
  Image_ProgressBar {
   widgetClassPath=&Image
   bIsVariable=true
   parent=Overlay_ProgressBar
   widget=Image_ProgressBar
   properties {
    visibility=SelfHitTestInvisible
    brush {
     bIsDynamicallyLoaded=false
     drawAs=Image
     imageType=FullColor
     mirroring=NoMirror
     resourceName=None
     resourceObject=/TidalCraneSharedActivitiesUI/Materials/M_UI_Reclamation_Progress
     tiling=NoTile
     imageSize=<700 70>
     margin=<ltrb 0 0 0 0>
     outlineSettings {
      bUseBrushTransparency=false
      roundingType=HalfHeightRadius
      width=0
      color {colorUseRule=UseColor_Specified, specifiedColor=<rgba 0 0 0 0>}
      cornerRadii=<0 0 0 0>
     }
     tintColor {colorUseRule=UseColor_Specified, specifiedColor=<rgba 1 1 1 1>}
     uVRegion {bIsValid=false, max=<0 0>, min=<0 0>}
    }
   }
   slot {class=OverlaySlot}
  }
  Image_Glow {
   widgetClassPath=&Image
   bIsVariable=true
   parent=Overlay_ProgressBar
   widget=Image_Glow
   properties {
    visibility=SelfHitTestInvisible
    brush {
     bIsDynamicallyLoaded=false
     drawAs=Image
     imageType=FullColor
     mirroring=NoMirror
     resourceName=None
     resourceObject=/TidalCraneSharedActivitiesUI/Materials/MI_UI_Reclamation_Progress_Glow
     tiling=NoTile
     imageSize=<32 32>
     margin=<ltrb 0 0 0 0>
     outlineSettings {
      bUseBrushTransparency=false
      roundingType=HalfHeightRadius
      width=0
      color {colorUseRule=UseColor_Specified, specifiedColor=<rgba 0 0 0 0>}
      cornerRadii=<0 0 0 0>
     }
     tintColor {colorUseRule=UseColor_Specified, specifiedColor=<rgba 1 1 1 1>}
     uVRegion {bIsValid=false, max=<0 0>, min=<0 0>}
    }
    renderTransform {angle=0, scale=<1 1>, shear=<0 0>, translation=<0 -32>}
    renderTransformPivot=<0.5 1>
   }
   slot {class=OverlaySlot, properties {horizontalAlignment=HAlign_Fill, verticalAlignment=VAlign_Fill}}
  }
  Image_BossGlow {
   widgetClassPath=&Image
   bIsVariable=true
   parent=Overlay_ProgressBar
   widget=Image_BossGlow
   properties {
    renderOpacity=0.025
    visibility=SelfHitTestInvisible
    brush {
     bIsDynamicallyLoaded=false
     drawAs=Image
     imageType=FullColor
     mirroring=NoMirror
     resourceName=None
     resourceObject=/TidalCraneSharedActivitiesUI/Materials/MI_UI_HardEdgedGlow_Purple
     tiling=NoTile
     imageSize=<32 32>
     margin=<ltrb 0 0 0 0>
     outlineSettings {
      bUseBrushTransparency=false
      roundingType=HalfHeightRadius
      width=0
      color {colorUseRule=UseColor_Specified, specifiedColor=<rgba 0 0 0 0>}
      cornerRadii=<0 0 0 0>
     }
     tintColor {colorUseRule=UseColor_Specified, specifiedColor=<rgba 1 1 1 1>}
     uVRegion {bIsValid=false, max=<0 0>, min=<0 0>}
    }
    renderTransform {angle=0, scale=<1 1.4>, shear=<0 0>, translation=<0 0>}
   }
   slot {class=OverlaySlot, properties {horizontalAlignment=HAlign_Fill, verticalAlignment=VAlign_Fill}}
  }
  Overlay_71 {widgetClassPath=&Overlay, parent=VerticalBox_All, widget=Overlay_71, slot {class=VerticalBoxSlot}}
  Text_Title {
   widgetClassPath=&CommonTextBlock
   bIsVariable=true
   parent=Overlay_71
   widget=Text_Title
   properties {
    mobileFontSizeMultiplier=1.2
    text="{objective message}"
    textTransformPolicy=ToUpper
    wrapTextAt=680
    font {
     bForceMonospaced=false
     bMaterialIsStencil=false
     fontMaterial=null
     fontObject=&BurbankSmall
     letterSpacing=0
     monospacedWidth=1
     size=19.5
     skewAmount=0
     typefaceFontName=Black
     outlineSettings {
      bApplyOutlineToDropShadows=false
      bMiteredCorners=false
      bSeparateFillAlpha=false
      outlineMaterial=&M-FontShadow
      outlineSize=3
      outlineColor=<rgba 0 0 0 0.5>
     }
    }
   }
   slot {class=OverlaySlot, properties {horizontalAlignment=HAlign_Center, verticalAlignment=VAlign_Fill}}
  }
  Image_TextFX {
   widgetClassPath=&Image
   bIsVariable=true
   parent=Overlay_71
   widget=Image_TextFX
   properties {
    renderOpacity=0
    visibility=SelfHitTestInvisible
    brush {
     bIsDynamicallyLoaded=false
     drawAs=Image
     imageType=FullColor
     mirroring=NoMirror
     resourceName=None
     resourceObject=/TidalCraneSharedActivitiesUI/Materials/M_UI_TextBlink
     tiling=NoTile
     imageSize=<32 32>
     margin=<ltrb 0 0 0 0>
     outlineSettings {
      bUseBrushTransparency=false
      roundingType=HalfHeightRadius
      width=0
      color {colorUseRule=UseColor_Specified, specifiedColor=<rgba 0 0 0 0>}
      cornerRadii=<0 0 0 0>
     }
     tintColor {colorUseRule=UseColor_Specified, specifiedColor=<rgba 1 1 1 1>}
     uVRegion {bIsValid=false, max=<0 0>, min=<0 0>}
    }
   }
   slot {class=OverlaySlot, properties {horizontalAlignment=HAlign_Fill, verticalAlignment=VAlign_Fill}}
  }
  TutorialHighlight {widgetClassPath=&TutorialHighlight}
 ]
}
